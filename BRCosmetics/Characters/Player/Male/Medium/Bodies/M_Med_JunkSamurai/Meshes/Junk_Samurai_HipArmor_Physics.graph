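# Physics asset for the Junk Samurai male body's hip armor, giving the left and right armor plates physical motion.
# Each side is a chain of box bodies from pelvis through a master bone to three plate segments, joined with tight twist and modest swing limits.
# Pelvis and thighs use stacked capsules, likely as collision volumes for the plates, with thigh constraints left free in swing and twist.
# Left and right sides use slightly different limits and thigh capsule sizes.
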PhysicsAsset /BRCosmetics/Characters/Player/Male/Medium/Bodies/M_Med_JunkSamurai/Meshes/Junk_Samurai_HipArmor_Physics
bodies {
 dyn_hip_armor_01_l [{shapeType=Box, extentX=6.2576, extentY=4.7807, extentZ=21.6005, center=<1.9102 -0.4532 0.1762>, rotation=<rot 0 0 79.3466>}]
 dyn_hip_armor_01_r [{shapeType=Box, extentX=6.2576, extentY=4.7807, extentZ=21.6005, center=<1.9102 -0.4532 0.1762>, rotation=<rot 0 0 90.5737>}]
 dyn_hip_armor_02_l [{shapeType=Box, extentX=5.7576, extentY=5.7807, extentZ=21.6005, center=<2.522 -0.4532 0.1762>, rotation=<rot 0.0001 0 83.9678>}]
 dyn_hip_armor_02_r [{shapeType=Box, extentX=5.5277, extentY=5.5508, extentZ=21.6005, center=<2.5138 -0.4532 0.1762>, rotation=<rot 0 0 96.4157>}]
 dyn_hip_armor_03_l [{shapeType=Box, extentX=7.7576, extentY=5.7807, extentZ=21.6005, center=<3.6081 -0.4531 0.1762>, rotation=<rot 0.0001 0 80.4885>}]
 dyn_hip_armor_03_r [{shapeType=Box, extentX=7.7576, extentY=5.7807, extentZ=21.6005, center=<3.6081 -0.4531 0.1762>, rotation=<rot 0.0001 0 88.7123>}]
 dyn_hip_armor_mstr_l [{shapeType=Box, extentX=14.7576, extentY=4.7807, extentZ=21.6005, center=<5.6593 -0.4532 0.1762>, rotation=<rot 0 0 90>}]
 dyn_hip_armor_mstr_r [{shapeType=Box, extentX=14.7576, extentY=4.2807, extentZ=21.6005, center=<5.6593 -0.4532 0.1762>, rotation=<rot 0 0 90>}]
 pelvis [{shapeType=Capsule, length=10.7728, radius=7.2636}, {shapeType=Capsule, length=11.6752, radius=7.2636, center=<-0.0001 -6.876 -0.0003>}]
 thigh_l [
  {shapeType=Capsule, length=5.8763, radius=8.6087, center=<-23.8168 -0.0002 -0.7593>, rotation=<rot 0 0 90>}
  {shapeType=Capsule, length=6.8763, radius=7.3587, center=<-13.8168 -0.0002 -1.5896>, rotation=<rot 0 0 90>}
  {shapeType=Capsule, length=10.9382, radius=6.6087, center=<-3.8168 -0.0002 -1.2979>, rotation=<rot 0 0 90>}
 ]
 thigh_r [
  {shapeType=Capsule, length=2.8763, radius=11.3587, center=<24.133 -0.0007 0>, rotation=<rot 0 0 90>}
  {shapeType=Capsule, length=4.3763, radius=9.8587, center=<14.2802 -0.0009 0>, rotation=<rot 0 0 90>}
  {shapeType=Capsule, length=8.6903, radius=8.8587, center=<5.3072 -0.0006 0>, rotation=<rot 0 0 90>}
 ]
}
constraints [
 {bone1Name=dyn_hip_armor_01_l, bone2Name=dyn_hip_armor_mstr_l, swing1LimitDegrees=20, swing2LimitDegrees=15, twistLimitDegrees=5}
 {bone1Name=dyn_hip_armor_02_l, bone2Name=dyn_hip_armor_01_l, swing1LimitDegrees=20, swing2LimitDegrees=15, twistLimitDegrees=5}
 {bone1Name=dyn_hip_armor_03_l, bone2Name=dyn_hip_armor_02_l, swing1LimitDegrees=20, swing2LimitDegrees=15, twistLimitDegrees=5}
 {bone1Name=dyn_hip_armor_01_r, bone2Name=dyn_hip_armor_mstr_r, swing1LimitDegrees=20, swing2LimitDegrees=20, twistLimitDegrees=5}
 {bone1Name=dyn_hip_armor_03_r, bone2Name=dyn_hip_armor_02_r, swing1LimitDegrees=20, swing2LimitDegrees=20, twistLimitDegrees=5}
 {bone1Name=dyn_hip_armor_02_r, bone2Name=dyn_hip_armor_01_r, swing1LimitDegrees=20, swing2LimitDegrees=20, twistLimitDegrees=5}
 {bone1Name=dyn_hip_armor_mstr_l, bone2Name=pelvis, swing1LimitDegrees=15, swing2LimitDegrees=35, twistLimitDegrees=0}
 {bone1Name=dyn_hip_armor_mstr_r, bone2Name=pelvis, swing1LimitDegrees=40, swing2LimitDegrees=40, twistLimitDegrees=1}
 {
  bone1Name=thigh_l
  bone2Name=pelvis
  swing1LimitDegrees=45
  swing1Motion=Free
  swing2LimitDegrees=45
  swing2Motion=Free
  twistLimitDegrees=45
  twistMotion=Free
 }
 {
  bone1Name=thigh_r
  bone2Name=pelvis
  swing1LimitDegrees=45
  swing1Motion=Free
  swing2LimitDegrees=45
  swing2Motion=Free
  twistLimitDegrees=45
  twistMotion=Free
 }
]
