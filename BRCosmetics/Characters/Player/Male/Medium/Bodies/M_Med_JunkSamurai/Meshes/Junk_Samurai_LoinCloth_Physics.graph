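# Physics asset for the Junk Samurai loincloth, defining front and back chains of dynamic cloth bones hung from the pelvis.
# The front chain uses four capsules and the back three boxes, linked by constraints with narrow swing1 and twist but wide swing2, left free at the back tip.
# Pelvis and thigh capsules, three stacked per thigh, form the body volumes around the cloth, with thighs constrained freely to the pelvis.
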PhysicsAsset /BRCosmetics/Characters/Player/Male/Medium/Bodies/M_Med_JunkSamurai/Meshes/Junk_Samurai_LoinCloth_Physics
bodies {
 dyn_back_loincloth_01 [{shapeType=Box, extentX=9.8845, extentY=3.5545, extentZ=16.6234, center=<3.9644 0.0569 -0.1872>, rotation=<rot 0 -0.6317 -87.4358>}]
 dyn_back_loincloth_02 [{shapeType=Box, extentX=9.3845, extentY=3.5545, extentZ=16.6234, center=<3.2851 0.0644 -0.1872>, rotation=<rot 0 -0.6317 -87.4358>}]
 dyn_back_loincloth_03 [{shapeType=Box, extentX=9.8845, extentY=3.5545, extentZ=16.6234, center=<3.9645 0.0637 -0.0354>, rotation=<rot -1.5336 -0.7004 -87.4349>}]
 dyn_front_loincloth_01 [{shapeType=Capsule, length=11.0344, radius=2.4864, center=<7.2574 0.0011 -0.1311>, rotation=<rot 0 0 -90.4003>}]
 dyn_front_loincloth_02 [{shapeType=Capsule, length=11.7597, radius=2.8401, center=<3.7442 -0.002 0.2896>, rotation=<rot 0 0 -90.4003>}]
 dyn_front_loincloth_03 [{shapeType=Capsule, length=13.0162, radius=3.3807, center=<3.7442 -0.0006 0.0984>, rotation=<rot 0 0 -90.4003>}]
 dyn_front_loincloth_04 [{shapeType=Capsule, length=12.6726, radius=3.7243, center=<3.7442 -0.0006 0.0984>, rotation=<rot 0 0 -90.4003>}]
 pelvis [{shapeType=Capsule, length=10.7728, radius=7.2636}]
 thigh_l [
  {shapeType=Capsule, length=6.8805, radius=8.6087, center=<-23.8168 -0.0002 -0.7593>, rotation=<rot 0 0 90>}
  {shapeType=Capsule, length=8.4517, radius=7.3587, center=<-13.8168 -0.0002 -1.5896>, rotation=<rot 0 0 90>}
  {shapeType=Capsule, length=7.8763, radius=6.6087, center=<-3.8168 -0.0002 -1.2979>, rotation=<rot 0 0 90>}
 ]
 thigh_r [
  {shapeType=Capsule, length=2.8763, radius=11.3587, center=<24.133 -0.0007 0>, rotation=<rot 0 0 90>}
  {shapeType=Capsule, length=4.3763, radius=9.8587, center=<14.2802 -0.0009 0>, rotation=<rot 0 0 90>}
  {shapeType=Capsule, length=5.3763, radius=8.8587, center=<5.3072 -0.0006 0>, rotation=<rot 0 0 90>}
 ]
}
constraints [
 {bone1Name=dyn_front_loincloth_02, bone2Name=dyn_front_loincloth_01, swing1LimitDegrees=10, swing2LimitDegrees=55, twistLimitDegrees=5}
 {bone1Name=dyn_front_loincloth_03, bone2Name=dyn_front_loincloth_02, swing1LimitDegrees=10, swing2LimitDegrees=55, twistLimitDegrees=2.5}
 {bone1Name=dyn_back_loincloth_02, bone2Name=dyn_back_loincloth_01, swing1LimitDegrees=10, swing2LimitDegrees=45, twistLimitDegrees=10}
 {
  bone1Name=dyn_back_loincloth_03
  bone2Name=dyn_back_loincloth_02
  swing1LimitDegrees=10
  swing2LimitDegrees=60
  swing2Motion=Free
  twistLimitDegrees=10
 }
 {bone1Name=dyn_front_loincloth_01, bone2Name=pelvis, swing1LimitDegrees=10, swing2LimitDegrees=55, twistLimitDegrees=5}
 {bone1Name=dyn_back_loincloth_01, bone2Name=pelvis, swing1LimitDegrees=10, swing2LimitDegrees=50, twistLimitDegrees=10}
 {
  bone1Name=thigh_l
  bone2Name=pelvis
  swing1LimitDegrees=45
  swing1Motion=Free
  swing2LimitDegrees=45
  swing2Motion=Free
  twistLimitDegrees=45
  twistMotion=Free
 }
 {
  bone1Name=thigh_r
  bone2Name=pelvis
  swing1LimitDegrees=45
  swing1Motion=Free
  swing2LimitDegrees=45
  swing2Motion=Free
  twistLimitDegrees=45
  twistMotion=Free
 }
 {bone1Name=dyn_front_loincloth_04, bone2Name=dyn_front_loincloth_03, swing1LimitDegrees=5, swing2LimitDegrees=40, twistLimitDegrees=5}
]
