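# Physics asset for the male medium Junk Samurai body's loincloth, named as the crouch variant.
# Box bodies form a four-bone front chain and a three-bone back chain, both anchored to the pelvis.
# Capsules cover pelvis, thighs and calves, with angularly free leg joints.
# Front boxes flare wider and twist limits loosen down the chain; back joints are tighter, except one free swing axis at the last.
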PhysicsAsset /BRCosmetics/Characters/Player/Male/Medium/Bodies/M_Med_JunkSamurai/Meshes/Junk_Samurai_LoinCloth_Physics_Crouch
bodies {
 calf_l [{shapeType=Capsule, length=33.2238, radius=7.2347, center=<-19.2035 0.0487 0.0011>, rotation=<rot 0 90 90>}]
 calf_r [{shapeType=Capsule, length=31.9752, radius=7.2392, center=<17.7578 -0.048 0.0001>, rotation=<rot 0 90 90>}]
 dyn_back_loincloth_01 [{shapeType=Box, extentX=9.8845, extentY=3.5545, extentZ=16.6234, center=<3.9644 0.0569 -0.1872>, rotation=<rot 0 -0.6317 -87.4358>}]
 dyn_back_loincloth_02 [{shapeType=Box, extentX=9.3845, extentY=3.5545, extentZ=16.6234, center=<3.2851 0.0644 -0.1872>, rotation=<rot 0 -0.6317 -87.4358>}]
 dyn_back_loincloth_03 [{shapeType=Box, extentX=9.8845, extentY=3.5545, extentZ=16.6234, center=<3.9645 0.0637 -0.0354>, rotation=<rot -1.5336 -0.7004 -87.4349>}]
 dyn_front_loincloth_01 [{shapeType=Box, extentX=13.7126, extentY=8.8139, extentZ=1.6608, center=<6.8618 0.0007 0>}]
 dyn_front_loincloth_02 [{shapeType=Box, extentX=7.3474, extentY=13.0328, extentZ=1.6608, center=<3.755 0.0003 0>}]
 dyn_front_loincloth_03 [{shapeType=Box, extentX=7.3474, extentY=15.2098, extentZ=1.6608, center=<3.755 0.0003 0>}]
 dyn_front_loincloth_04 [{shapeType=Box, extentX=7.3474, extentY=17.3966, extentZ=1.6608, center=<3.755 0.0003 0>}]
 pelvis [{shapeType=Capsule, length=10.7498, radius=10.0089, center=<0 -0.3127 0>}]
 thigh_l [{shapeType=Capsule, length=20.8204, radius=9.6656, center=<-16.7792 0.0003 0>, rotation=<rot 0 90 90>}]
 thigh_r [{shapeType=Capsule, length=20.8204, radius=9.6656, center=<17.0491 -0.0001 0>, rotation=<rot 0 90 90>}]
}
constraints [
 {bone1Name=dyn_back_loincloth_02, bone2Name=dyn_back_loincloth_01, swing1LimitDegrees=10, swing2LimitDegrees=45, twistLimitDegrees=10}
 {
  bone1Name=dyn_back_loincloth_03
  bone2Name=dyn_back_loincloth_02
  swing1LimitDegrees=10
  swing2LimitDegrees=60
  swing2Motion=Free
  twistLimitDegrees=10
 }
 {bone1Name=dyn_back_loincloth_01, bone2Name=pelvis, swing1LimitDegrees=10, swing2LimitDegrees=50, twistLimitDegrees=10}
 {
  bone1Name=thigh_l
  bone2Name=pelvis
  swing1LimitDegrees=45
  swing1Motion=Free
  swing2LimitDegrees=45
  swing2Motion=Free
  twistLimitDegrees=45
  twistMotion=Free
 }
 {
  bone1Name=thigh_r
  bone2Name=pelvis
  swing1LimitDegrees=45
  swing1Motion=Free
  swing2LimitDegrees=45
  swing2Motion=Free
  twistLimitDegrees=45
  twistMotion=Free
 }
 {
  bone1Name=calf_r
  bone2Name=thigh_r
  swing1LimitDegrees=45
  swing1Motion=Free
  swing2LimitDegrees=45
  swing2Motion=Free
  twistLimitDegrees=45
  twistMotion=Free
 }
 {
  bone1Name=calf_l
  bone2Name=thigh_l
  swing1LimitDegrees=45
  swing1Motion=Free
  swing2LimitDegrees=45
  swing2Motion=Free
  twistLimitDegrees=45
  twistMotion=Free
 }
 {bone1Name=dyn_front_loincloth_01, bone2Name=pelvis, swing1LimitDegrees=30, swing2LimitDegrees=45, twistLimitDegrees=20}
 {bone1Name=dyn_front_loincloth_02, bone2Name=dyn_front_loincloth_01, swing1LimitDegrees=30, swing2LimitDegrees=45, twistLimitDegrees=30}
 {bone1Name=dyn_front_loincloth_03, bone2Name=dyn_front_loincloth_02, swing1LimitDegrees=30, swing2LimitDegrees=45, twistLimitDegrees=35}
 {bone1Name=dyn_front_loincloth_04, bone2Name=dyn_front_loincloth_03, swing1LimitDegrees=30, swing2LimitDegrees=45, twistLimitDegrees=45}
]
