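# Physics asset for the Junk Samurai male medium cosmetic's shoulder pads, giving the dynamic armor bones bodies and joint limits.
# Each side chains three small armor capsules from the clavicle (master, base 01, base 02), alongside larger clavicle and upper-arm capsules.
# Armor joints have zero twist limit and swing limits that tighten down the chain, with swing axes swapped between sides.
# Upper-arm to clavicle joints rotate freely.
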PhysicsAsset /BRCosmetics/Characters/Player/Male/Medium/Bodies/M_Med_JunkSamurai/Meshes/Junk_Samurai_ShoulderPad_Physics
bodies {
 clavicle_l [{shapeType=Capsule, length=0, radius=14}]
 clavicle_r [{shapeType=Capsule, length=0, radius=14}]
 dyn_shoulderArmor_base_l_01 [{shapeType=Capsule, length=10, radius=2.5, shapeName=dyn_shoulderArmor_base_l_01_capsule, center=<2 0 0>, rotation=<rot 0 0 90>}]
 dyn_shoulderArmor_base_l_02 [{shapeType=Capsule, length=10, radius=2.5, shapeName=dyn_shoulderArmor_base_l_02_capsule, center=<2 0 0>, rotation=<rot 0 0 90>}]
 dyn_shoulderArmor_base_r_01 [{shapeType=Capsule, length=10, radius=2.5, shapeName=dyn_shoulderArmor_base_r_01_capsule, center=<2 0 0>, rotation=<rot 0 -90 0>}]
 dyn_shoulderArmor_base_r_02 [{shapeType=Capsule, length=10, radius=2.5, shapeName=dyn_shoulderArmor_base_r_02_capsule, center=<2 0 0>, rotation=<rot 0 -90 0>}]
 dyn_shoulderArmor_master_l [{shapeType=Capsule, length=10, radius=2.5, shapeName=dyn_shoulderArmor_master_l_capsule, center=<2 0 0>, rotation=<rot 0 0 90>}]
 dyn_shoulderArmor_master_r [{shapeType=Capsule, length=10, radius=2.5, shapeName=dyn_shoulderArmor_master_r_capsule, center=<2 0 0>, rotation=<rot 0 -90 0>}]
 upperarm_l [{shapeType=Capsule, length=21, radius=4, center=<8.6514 0.0002 -0.5707>, rotation=<rot 90 0 0>}]
 upperarm_r [{shapeType=Capsule, length=21, radius=4, center=<-8.6514 0.0002 0.5707>, rotation=<rot 90 -2.5023 -2.3959>}]
}
constraints [
 {
  bone1Name=upperarm_r
  bone2Name=clavicle_r
  swing1LimitDegrees=45
  swing1Motion=Free
  swing2LimitDegrees=45
  swing2Motion=Free
  twistLimitDegrees=45
  twistMotion=Free
 }
 {
  bone1Name=upperarm_l
  bone2Name=clavicle_l
  swing1LimitDegrees=45
  swing1Motion=Free
  swing2LimitDegrees=45
  swing2Motion=Free
  twistLimitDegrees=45
  twistMotion=Free
 }
 {
  bone1Name=dyn_shoulderArmor_base_l_02
  bone2Name=dyn_shoulderArmor_base_l_01
  swing1LimitDegrees=5
  swing2LimitDegrees=5
  twistLimitDegrees=0
 }
 {
  bone1Name=dyn_shoulderArmor_base_l_01
  bone2Name=dyn_shoulderArmor_master_l
  swing1LimitDegrees=5
  swing2LimitDegrees=20
  twistLimitDegrees=0
 }
 {bone1Name=dyn_shoulderArmor_master_l, bone2Name=clavicle_l, swing1LimitDegrees=5, swing2LimitDegrees=50, twistLimitDegrees=0}
 {
  bone1Name=dyn_shoulderArmor_base_r_02
  bone2Name=dyn_shoulderArmor_base_r_01
  swing1LimitDegrees=5
  swing2LimitDegrees=5
  twistLimitDegrees=0
 }
 {
  bone1Name=dyn_shoulderArmor_base_r_01
  bone2Name=dyn_shoulderArmor_master_r
  swing1LimitDegrees=20
  swing2LimitDegrees=5
  twistLimitDegrees=0
 }
 {bone1Name=dyn_shoulderArmor_master_r, bone2Name=clavicle_r, swing1LimitDegrees=50, swing2LimitDegrees=5, twistLimitDegrees=0}
]
