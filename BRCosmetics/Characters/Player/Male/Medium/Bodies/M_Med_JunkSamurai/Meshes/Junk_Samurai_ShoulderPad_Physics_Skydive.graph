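# Physics asset for the Junk Samurai male body's shoulder pads, apparently the skydiving variant, simulating the dangling armor on both shoulders.
# Capsule bodies cover the clavicles, upper arms and a three-bone armor chain per side, running from master through two base bones.
# Each chain hangs from its clavicle with tight twist and wide Swing2 limits, slightly tighter on the left.
# The upper-arm-to-clavicle constraints leave all angular motion free.
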PhysicsAsset /BRCosmetics/Characters/Player/Male/Medium/Bodies/M_Med_JunkSamurai/Meshes/Junk_Samurai_ShoulderPad_Physics_Skydive
bodies {
 clavicle_l [{shapeType=Capsule, length=0, radius=13.8684, rotation=<rot 0 177.6778 -79.7147>}]
 clavicle_r [{shapeType=Capsule, length=0, radius=13.8684, rotation=<rot 0 177.6778 -79.7147>}]
 dyn_shoulderArmor_base_l_01 [{shapeType=Capsule, length=9.4266, radius=3.772, rotation=<rot 0 6.1493 -79.9031>}]
 dyn_shoulderArmor_base_l_02 [{shapeType=Capsule, length=9.4266, radius=3.772, rotation=<rot 0 6.1493 -79.9031>}]
 dyn_shoulderArmor_base_r_01 [{shapeType=Capsule, length=9.4266, radius=3.772, rotation=<rot 0 0.0001 -5.8569>}]
 dyn_shoulderArmor_base_r_02 [{shapeType=Capsule, length=9.4266, radius=3.772}]
 dyn_shoulderArmor_master_l [{shapeType=Capsule, length=9.4266, radius=3.772, rotation=<rot 0 6.1493 -79.9031>}]
 dyn_shoulderArmor_master_r [{shapeType=Capsule, length=9.4266, radius=3.772}]
 upperarm_l [{shapeType=Capsule, length=21.1372, radius=4.1471, center=<13.1015 0.0006 0>, rotation=<rot 90 0 0>}]
 upperarm_r [{shapeType=Capsule, length=21.1372, radius=4.1471, center=<-12.4369 0.0002 0.5707>, rotation=<rot 90 -2.5023 -2.3959>}]
}
constraints [
 {
  bone1Name=upperarm_r
  bone2Name=clavicle_r
  swing1LimitDegrees=45
  swing1Motion=Free
  swing2LimitDegrees=45
  swing2Motion=Free
  twistLimitDegrees=45
  twistMotion=Free
 }
 {
  bone1Name=upperarm_l
  bone2Name=clavicle_l
  swing1LimitDegrees=45
  swing1Motion=Free
  swing2LimitDegrees=45
  swing2Motion=Free
  twistLimitDegrees=45
  twistMotion=Free
 }
 {bone1Name=dyn_shoulderArmor_master_l, bone2Name=clavicle_l, swing1LimitDegrees=5, swing2LimitDegrees=50, twistLimitDegrees=5}
 {bone1Name=dyn_shoulderArmor_master_r, bone2Name=clavicle_r, swing1LimitDegrees=5, swing2LimitDegrees=80, twistLimitDegrees=5}
 {
  bone1Name=dyn_shoulderArmor_base_l_01
  bone2Name=dyn_shoulderArmor_master_l
  swing1LimitDegrees=20
  swing2LimitDegrees=80
  twistLimitDegrees=5
 }
 {
  bone1Name=dyn_shoulderArmor_base_l_02
  bone2Name=dyn_shoulderArmor_base_l_01
  swing1LimitDegrees=30
  swing2LimitDegrees=80
  twistLimitDegrees=5
 }
 {
  bone1Name=dyn_shoulderArmor_base_r_01
  bone2Name=dyn_shoulderArmor_master_r
  swing1LimitDegrees=30
  swing2LimitDegrees=80
  twistLimitDegrees=5
 }
 {
  bone1Name=dyn_shoulderArmor_base_r_02
  bone2Name=dyn_shoulderArmor_base_r_01
  swing1LimitDegrees=30
  swing2LimitDegrees=80
  twistLimitDegrees=5
 }
]
