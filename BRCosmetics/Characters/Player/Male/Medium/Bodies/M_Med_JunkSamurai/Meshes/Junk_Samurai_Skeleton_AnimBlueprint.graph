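# Animation Blueprint for the Junk Samurai player body that adds procedural motion and pose correction to its shoulder armor, hip armor and loincloths.
# From a cached legs pose, shoulder pads are constrained to arm and clavicle bones by an LOD alpha, then simulated with spring-driven AnimDynamics.
# Additive bone offsets follow lookup and crouch variables.
# Near-duplicate chains are bool-blended for skydiving, parachute and sprint states before hip-armor raise offsets apply.
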AnimBlueprint /BRCosmetics/Characters/Player/Male/Medium/Bodies/M_Med_JunkSamurai/Meshes/Junk_Samurai_Skeleton_AnimBlueprint
target_skeleton=&/Junk_Samurai_Skeleton
graphs [
 AnimGraph {nodes [
   n1 = Anim Dynamics - Bone: dyn_shoulderArmor_base_r_01(Alpha=1, ExternalForce="0,0,0") {
    angularDampingOverride=1
    angularSpringConstant=20000
    bAngularSpring=true
    bOverrideAngularDamping=true
    bUseGravityOverride=true
    gravityScale=0.35
    simulationSpace=BoneRelative
    alphaBoolBlend {blendOption=Linear}
    boundBone {boneName=dyn_shoulderArmor_base_r_01}
    componentAppliedLinearAccClamp=<1000 1000 1000>
    componentLinearAccScale=<0.3 0.2 0.2>
    componentLinearVelScale=<0.3 0.2 0.2>
    gravityOverride {z=-980}
    physicsBodyDefinitions [{
      boundBone {boneName=dyn_shoulderArmor_base_r_01}
      boxExtents=<20 30 30>
      constraintSetup {angularLimitsMax=<2 0 10>, angularLimitsMin=<-2 0 -10>, angularTarget=<1 0 0>}
      localJointOffset=<4 0 0>
     }]
    relativeSpaceBone {boneName=spine_05}
   }
   n2 = Anim Dynamics - Bone: dyn_shoulderArmor_base_r_02(ComponentPose=@n1, Alpha=1, ExternalForce="0,0,0") {
    angularDampingOverride=1
    angularSpringConstant=20000
    bAngularSpring=true
    bOverrideAngularDamping=true
    bUseGravityOverride=true
    gravityScale=0.35
    simulationSpace=BoneRelative
    alphaBoolBlend {blendOption=Linear}
    boundBone {boneName=dyn_shoulderArmor_base_r_02}
    componentAppliedLinearAccClamp=<1000 1000 1000>
    componentLinearAccScale=<0.3 0.2 0.2>
    componentLinearVelScale=<0.3 0.2 0.2>
    gravityOverride {z=-980}
    physicsBodyDefinitions [{
      boundBone {boneName=dyn_shoulderArmor_base_r_02}
      boxExtents=<20 30 30>
      constraintSetup {angularLimitsMax=<2 0 7>, angularLimitsMin=<-2 0 -7>, angularTarget=<1 0 0>}
      localJointOffset=<4 0 0>
     }]
    relativeSpaceBone {boneName=spine_05}
   }
   Anim Dynamics - Bone: dyn_shoulderArmor_base_r_03(ComponentPose=@n2, Alpha=1, ExternalForce="0,0,0") {
    angularDampingOverride=1
    angularSpringConstant=20000
    bAngularSpring=true
    bOverrideAngularDamping=true
    bUseGravityOverride=true
    gravityScale=0.35
    simulationSpace=BoneRelative
    alphaBoolBlend {blendOption=Linear}
    boundBone {boneName=dyn_shoulderArmor_base_r_03}
    componentAppliedLinearAccClamp=<1000 1000 1000>
    componentLinearAccScale=<0.3 0.2 0.2>
    componentLinearVelScale=<0.3 0.2 0.2>
    gravityOverride {z=-980}
    physicsBodyDefinitions [{
      boundBone {boneName=dyn_shoulderArmor_base_r_03}
      boxExtents=<20 30 30>
      constraintSetup {angularLimitsMax=<2 0 5>, angularLimitsMin=<-2 0 -5>, angularTarget=<1 0 0>}
      localJointOffset=<4 0 0>
     }]
    relativeSpaceBone {boneName=spine_05}
   }
   "Constraint - dyn_hip_armor_mstr_r_offset (TTT)"!(ComponentPose, Alpha=1, ConstraintWeights_0=0.0365, ConstraintWeights_1:float=0.2559, ConstraintWeights_2:float=0.7076) -> (Pose) {
    alphaBoolBlend {blendOption=Linear}
    boneToModify {boneName=dyn_hip_armor_mstr_r_offset}
    constraintSetup [
     {transformType=Translation, targetBone {boneName=spine_01}}
     {transformType=Translation, targetBone {boneName=thigh_twist_01_r}}
     {transformType=Translation, targetBone {boneName=pelvis}}
    ]
    constraintWeights [1,1,1]
   }
   "Constraint - dyn_hip_armor_mstr_r_offset (R)"(Alpha=1, ConstraintWeights_0=1) {
    alphaBoolBlend {blendOption=Linear}
    boneToModify {boneName=dyn_hip_armor_mstr_r_offset}
    constraintSetup [{transformType=Rotation, targetBone {boneName=thigh_r}}]
    constraintWeights [1]
   }
   "Transform (Modify) Bone - Bone: dyn_hip_armor_mstr_r_offset"(Alpha=1, Translation="-3.5,-5,2", Rotation="-15,-20,10", Scale="1,1,1") {
    rotationMode=BMM_Additive
    translationMode=BMM_Additive
    alphaBoolBlend {blendOption=Linear}
    boneToModify {boneName=dyn_hip_armor_mstr_r_offset}
    rotation=<rot -15 -20 10>
    translation=<-3.5 -5 2>
   }
   "Transform (Modify) Bone - Bone: dyn_hip_armor_mstr_r_offset"(Alpha=1, Translation="-3.5,-5,2", Rotation="-15,-20,10", Scale="1,1,1") {
    rotationMode=BMM_Additive
    translationMode=BMM_Additive
    alphaBoolBlend {blendOption=Linear}
    boneToModify {boneName=dyn_hip_armor_mstr_r_offset}
    rotation=<rot -15 -20 10>
    translation=<-3.5 -5 2>
   }
   "Transform (Modify) Bone - Bone: dyn_front_loincloth_02"(Alpha=1, Translation="0,10,0", Rotation="0,0,-45", Scale="1,1,1") {
    rotationMode=BMM_Additive
    translationMode=BMM_Additive
    alphaBoolBlend {blendOption=Linear}
    boneToModify {boneName=dyn_front_loincloth_02}
    rotation {roll=-45}
    translation {y=10}
   }
   RigidBody(Alpha=0.25) {
    bOverrideWorldGravity=true
    overridePhysicsAsset=&/Junk_Samurai_ShoulderPad_Physics_Skydive
    simulationSpace=BaseBoneSpace
    alphaBoolBlend {blendOption=Linear}
    baseBoneRef {boneName=pelvis}
    componentAppliedLinearAccClamp=<2500 2500 2500>
    componentLinearAccScale=<0.65 0.5 0.5>
    componentLinearVelScale=<0.65 0.5 0.5>
    overrideWorldGravity {z=-980}
   }
   RigidBody(Alpha=0.25) {
    bOverrideWorldGravity=true
    overridePhysicsAsset=&Junk_Samurai_ShoulderPad_Physics_Skydive
    simulationSpace=BaseBoneSpace
    alphaBoolBlend {blendOption=Linear}
    baseBoneRef {boneName=pelvis}
    componentAppliedLinearAccClamp=<2500 2500 2500>
    componentLinearAccScale=<0.65 0.5 0.5>
    componentLinearVelScale=<0.65 0.5 0.5>
    overrideWorldGravity {z=-980}
   }
   RigidBody(Alpha=0.25) {
    bOverrideWorldGravity=true
    overridePhysicsAsset=&/Junk_Samurai_ShoulderPad_Physics
    simulationSpace=BaseBoneSpace
    alphaBoolBlend {blendOption=Linear}
    baseBoneRef {boneName=pelvis}
    componentAppliedLinearAccClamp=<2500 2500 2500>
    componentLinearAccScale=<0.65 0.5 0.5>
    componentLinearVelScale=<0.65 0.5 0.5>
    overrideWorldGravity {z=-980}
   }
   RigidBody(Alpha=0.25) {
    bOverrideWorldGravity=true
    overridePhysicsAsset=&Junk_Samurai_ShoulderPad_Physics
    simulationSpace=BaseBoneSpace
    alphaBoolBlend {blendOption=Linear}
    baseBoneRef {boneName=pelvis}
    componentAppliedLinearAccClamp=<2500 2500 2500>
    componentLinearAccScale=<0.65 0.5 0.5>
    componentLinearVelScale=<0.65 0.5 0.5>
    overrideWorldGravity {z=-980}
   }
   n3 = Get LOD_Switch() -> (LOD_Switch:double)
   n4 = "Constraint - dyn_shoulderArmor_master_l_Offset (PP)"!(
    ComponentPose=Local To Component(LocalPose=Use cached pose 'CachedLegs'())
    Alpha=@n3
    ConstraintWeights_0=0.7779
    ConstraintWeights_1:float=0.2221
   ) -> (Pose) {
    alphaBoolBlend {blendOption=Linear}
    boneToModify {boneName=dyn_shoulderArmor_master_l_Offset}
    constraintSetup [{targetBone {boneName=upperarm_twist_01_l}}, {targetBone {boneName=clavicle_l}}]
    constraintWeights [1,1]
   }
   n5 = "Constraint - dyn_shoulderArmor_master_r_Offset (PP)"!(ComponentPose=@n4, Alpha=@n3, ConstraintWeights_0=0.7779, ConstraintWeights_1:float=0.2221) -> (Pose) {
    alphaBoolBlend {blendOption=Linear}
    boneToModify {boneName=dyn_shoulderArmor_master_r_Offset}
    constraintSetup [{targetBone {boneName=upperarm_twist_01_r}}, {targetBone {boneName=clavicle_r}}]
    constraintWeights [1,1]
   }
   n6 = Anim Dynamics - Bone: dyn_shoulderArmor_base_l_01(ComponentPose=@n5, Alpha=1, ExternalForce="0,0,0") {
    angularDampingOverride=1
    angularSpringConstant=20000
    bAngularSpring=true
    bOverrideAngularDamping=true
    bUseGravityOverride=true
    gravityScale=0.35
    simulationSpace=BoneRelative
    alphaBoolBlend {blendOption=Linear}
    boundBone {boneName=dyn_shoulderArmor_base_l_01}
    componentAppliedLinearAccClamp=<1000 1000 1000>
    componentLinearAccScale=<0.3 0.2 0.2>
    componentLinearVelScale=<0.3 0.2 0.2>
    gravityOverride {z=-980}
    physicsBodyDefinitions [{
      boundBone {boneName=dyn_shoulderArmor_base_l_01}
      boxExtents=<20 30 30>
      constraintSetup {angularLimitsMax=<0 10 2>, angularLimitsMin=<0 -10 -2>, angularTarget=<1 0 0>}
      localJointOffset=<4 0 0>
     }]
    relativeSpaceBone {boneName=spine_05}
   }
   n7 = Anim Dynamics - Bone: dyn_shoulderArmor_base_l_02(ComponentPose=@n6, Alpha=1, ExternalForce="0,0,0") {
    angularDampingOverride=1
    angularSpringConstant=20000
    bAngularSpring=true
    bOverrideAngularDamping=true
    bUseGravityOverride=true
    gravityScale=0.35
    simulationSpace=BoneRelative
    alphaBoolBlend {blendOption=Linear}
    boundBone {boneName=dyn_shoulderArmor_base_l_02}
    componentAppliedLinearAccClamp=<1000 1000 1000>
    componentLinearAccScale=<0.3 0.2 0.2>
    componentLinearVelScale=<0.3 0.2 0.2>
    gravityOverride {z=-980}
    physicsBodyDefinitions [{
      boundBone {boneName=dyn_shoulderArmor_base_l_02}
      boxExtents=<20 30 30>
      constraintSetup {angularLimitsMax=<0 7 2>, angularLimitsMin=<0 -7 -2>, angularTarget=<1 0 0>}
      localJointOffset=<4 0 0>
     }]
    relativeSpaceBone {boneName=spine_05}
   }
   n8 = Anim Dynamics - Bone: dyn_shoulderArmor_base_l_03(ComponentPose=@n7, Alpha=1, ExternalForce="0,0,0") {
    angularDampingOverride=1
    angularSpringConstant=20000
    bAngularSpring=true
    bOverrideAngularDamping=true
    bUseGravityOverride=true
    gravityScale=0.35
    simulationSpace=BoneRelative
    alphaBoolBlend {blendOption=Linear}
    boundBone {boneName=dyn_shoulderArmor_base_l_03}
    componentAppliedLinearAccClamp=<1000 1000 1000>
    componentLinearAccScale=<0.3 0.2 0.2>
    componentLinearVelScale=<0.3 0.2 0.2>
    gravityOverride {z=-980}
    physicsBodyDefinitions [{
      boundBone {boneName=dyn_shoulderArmor_base_l_03}
      boxExtents=<20 30 30>
      constraintSetup {angularLimitsMax=<0 5 2>, angularLimitsMin=<0 -5 -2>, angularTarget=<1 0 0>}
      localJointOffset=<4 0 0>
     }]
    relativeSpaceBone {boneName=spine_05}
   }
   n9 = Anim Dynamics - Bone: dyn_shoulderArmor_base_r_01(ComponentPose=@n8, Alpha=1, ExternalForce="0,0,0") {
    angularDampingOverride=1
    angularSpringConstant=20000
    bAngularSpring=true
    bOverrideAngularDamping=true
    bUseGravityOverride=true
    gravityScale=0.35
    simulationSpace=BoneRelative
    alphaBoolBlend {blendOption=Linear}
    boundBone {boneName=dyn_shoulderArmor_base_r_01}
    componentAppliedLinearAccClamp=<1000 1000 1000>
    componentLinearAccScale=<0.3 0.2 0.2>
    componentLinearVelScale=<0.3 0.2 0.2>
    gravityOverride {z=-980}
    physicsBodyDefinitions [{
      boundBone {boneName=dyn_shoulderArmor_base_r_01}
      boxExtents=<20 30 30>
      constraintSetup {angularLimitsMax=<2 0 10>, angularLimitsMin=<-2 0 -10>, angularTarget=<1 0 0>}
      localJointOffset=<4 0 0>
     }]
    relativeSpaceBone {boneName=spine_05}
   }
   n10 = Anim Dynamics - Bone: dyn_shoulderArmor_base_r_02(ComponentPose=@n9, Alpha=1, ExternalForce="0,0,0") {
    angularDampingOverride=1
    angularSpringConstant=20000
    bAngularSpring=true
    bOverrideAngularDamping=true
    bUseGravityOverride=true
    gravityScale=0.35
    simulationSpace=BoneRelative
    alphaBoolBlend {blendOption=Linear}
    boundBone {boneName=dyn_shoulderArmor_base_r_02}
    componentAppliedLinearAccClamp=<1000 1000 1000>
    componentLinearAccScale=<0.3 0.2 0.2>
    componentLinearVelScale=<0.3 0.2 0.2>
    gravityOverride {z=-980}
    physicsBodyDefinitions [{
      boundBone {boneName=dyn_shoulderArmor_base_r_02}
      boxExtents=<20 30 30>
      constraintSetup {angularLimitsMax=<2 0 7>, angularLimitsMin=<-2 0 -7>, angularTarget=<1 0 0>}
      localJointOffset=<4 0 0>
     }]
    relativeSpaceBone {boneName=spine_05}
   }
   n11 = Anim Dynamics - Bone: dyn_shoulderArmor_base_r_03(ComponentPose=@n10, Alpha=1, ExternalForce="0,0,0") {
    angularDampingOverride=1
    angularSpringConstant=20000
    bAngularSpring=true
    bOverrideAngularDamping=true
    bUseGravityOverride=true
    gravityScale=0.35
    simulationSpace=BoneRelative
    alphaBoolBlend {blendOption=Linear}
    boundBone {boneName=dyn_shoulderArmor_base_r_03}
    componentAppliedLinearAccClamp=<1000 1000 1000>
    componentLinearAccScale=<0.3 0.2 0.2>
    componentLinearVelScale=<0.3 0.2 0.2>
    gravityOverride {z=-980}
    physicsBodyDefinitions [{
      boundBone {boneName=dyn_shoulderArmor_base_r_03}
      boxExtents=<20 30 30>
      constraintSetup {angularLimitsMax=<2 0 5>, angularLimitsMin=<-2 0 -5>, angularTarget=<1 0 0>}
      localJointOffset=<4 0 0>
     }]
    relativeSpaceBone {boneName=spine_05}
   }
   n12 = "Transform (Modify) Bone - Bone: dyn_shoulderArmor_master_r_Offset"(
    ComponentPose=@n11
    Alpha=Get ShoulderPadOffset_LookUp_R() -> (ShoulderPadOffset_LookUp_R:double)
    Translation="0,-3,-2"
    Rotation="10,40,50"
    Scale="0,0,0"
   ) {
    rotationMode=BMM_Additive
    rotationSpace=BCS_BoneSpace
    scaleSpace=BCS_BoneSpace
    translationMode=BMM_Additive
    translationSpace=BCS_BoneSpace
    alphaBoolBlend {blendOption=Linear}
    boneToModify {boneName=dyn_shoulderArmor_master_r_Offset}
    rotation=<rot 10 40 50>
    scale=<0 0 0>
    translation {y=-3, z=-2}
   }
   n13 = "Transform (Modify) Bone - Bone: dyn_shoulderArmor_master_l_Offset"(
    ComponentPose=@n12
    Alpha=Get ShoulderPadOffset_LookUp_L() -> (ShoulderPadOffset_LookUp_L:double)
    Translation="0,0,0"
    Rotation="5,0,-25"
    Scale="0,0,0"
   ) {
    rotationMode=BMM_Additive
    rotationSpace=BCS_BoneSpace
    scaleSpace=BCS_BoneSpace
    translationMode=BMM_Additive
    translationSpace=BCS_BoneSpace
    alphaBoolBlend {blendOption=Linear}
    boneToModify {boneName=dyn_shoulderArmor_master_l_Offset}
    rotation {pitch=5, roll=-25}
    scale=<0 0 0>
   }
   n14 = "Transform (Modify) Bone - Bone: dyn_shoulderArmor_master_r_Offset"(ComponentPose=@n13, Alpha=1, Translation="0,0.5,0", Rotation="0,3,0", Scale="0,0,0") {
    rotationMode=BMM_Additive
    rotationSpace=BCS_BoneSpace
    scaleSpace=BCS_BoneSpace
    translationMode=BMM_Additive
    translationSpace=BCS_BoneSpace
    alphaBoolBlend {blendOption=Linear}
    boneToModify {boneName=dyn_shoulderArmor_master_r_Offset}
    rotation {yaw=3}
    scale=<0 0 0>
    translation {y=0.5}
   }
   n15 = "Transform (Modify) Bone - Bone: dyn_shoulderArmor_master_l_Offset"(ComponentPose=@n14, Alpha=1, Translation="-1,0,0.05", Rotation="10,2,0", Scale="0,0,0") {
    rotationMode=BMM_Additive
    rotationSpace=BCS_BoneSpace
    scaleSpace=BCS_BoneSpace
    translationMode=BMM_Additive
    translationSpace=BCS_BoneSpace
    alphaBoolBlend {blendOption=Linear}
    boneToModify {boneName=dyn_shoulderArmor_master_l_Offset}
    rotation {pitch=10, yaw=2}
    scale=<0 0 0>
    translation {x=-1, z=0.05}
   }
   n16 = Get Not_Crouching() -> (Not_Crouching:double)
   n17 = "Transform (Modify) Bone - Bone: dyn_back_loincloth_01"(ComponentPose=@n15, Alpha=@n16, Translation="0,0,-2.5", Rotation="0,0,0", Scale="1,1,1") {
    rotationMode=BMM_Additive
    translationMode=BMM_Additive
    alphaBoolBlend {blendOption=Linear}
    boneToModify {boneName=dyn_back_loincloth_01}
    translation {z=-2.5}
   }
   n18 = "Transform (Modify) Bone - Bone: dyn_hip_armor_mstr_r_offset"(ComponentPose=@n17, Alpha=@n16, Translation="0,0,1", Rotation="0,0,0", Scale="1,1,1") {
    rotationMode=BMM_Additive
    translationMode=BMM_Additive
    alphaBoolBlend {blendOption=Linear}
    boneToModify {boneName=dyn_hip_armor_mstr_r_offset}
    translation {z=1}
   }
   n19 = "Transform (Modify) Bone - Bone: dyn_hip_armor_mstr_l_offset"(ComponentPose=@n18, Alpha=@n16, Translation="0,0,1", Rotation="0,0,0", Scale="1,1,1") {
    rotationMode=BMM_Additive
    translationMode=BMM_Additive
    alphaBoolBlend {blendOption=Linear}
    boneToModify {boneName=dyn_hip_armor_mstr_l_offset}
    translation {z=1}
   }
   n20 = Get LOD_Switch()
   n21 = "Constraint - dyn_shoulderArmor_master_l_Offset (PP)"(
    ComponentPose=Local To Component(LocalPose=Use cached pose 'CachedLegs'())
    Alpha=@n20
    ConstraintWeights_0=0.7779
    ConstraintWeights_1=0.2221
   ) {
    alphaBoolBlend {blendOption=Linear}
    boneToModify {boneName=dyn_shoulderArmor_master_l_Offset}
    constraintSetup [{targetBone {boneName=upperarm_twist_01_l}}, {targetBone {boneName=clavicle_l}}]
    constraintWeights [1,1]
   }
   n22 = "Constraint - dyn_shoulderArmor_master_r_Offset (PP)"(ComponentPose=@n21, Alpha=@n20, ConstraintWeights_0=0.7779, ConstraintWeights_1=0.2221) {
    alphaBoolBlend {blendOption=Linear}
    boneToModify {boneName=dyn_shoulderArmor_master_r_Offset}
    constraintSetup [{targetBone {boneName=upperarm_twist_01_r}}, {targetBone {boneName=clavicle_r}}]
    constraintWeights [1,1]
   }
   n23 = Anim Dynamics - Bone: dyn_shoulderArmor_base_l_01(ComponentPose=@n22, Alpha=1, ExternalForce="0,0,0") {
    angularDampingOverride=1
    angularSpringConstant=20000
    bAngularSpring=true
    bOverrideAngularDamping=true
    bUseGravityOverride=true
    gravityScale=0.35
    simulationSpace=BoneRelative
    alphaBoolBlend {blendOption=Linear}
    boundBone {boneName=dyn_shoulderArmor_base_l_01}
    componentAppliedLinearAccClamp=<1000 1000 1000>
    componentLinearAccScale=<0.3 0.2 0.2>
    componentLinearVelScale=<0.3 0.2 0.2>
    gravityOverride {z=-980}
    physicsBodyDefinitions [{
      boundBone {boneName=dyn_shoulderArmor_base_l_01}
      boxExtents=<20 30 30>
      constraintSetup {angularLimitsMax=<0 10 2>, angularLimitsMin=<0 -10 -2>, angularTarget=<1 0 0>}
      localJointOffset=<4 0 0>
     }]
    relativeSpaceBone {boneName=spine_05}
   }
   n24 = Anim Dynamics - Bone: dyn_shoulderArmor_base_l_02(ComponentPose=@n23, Alpha=1, ExternalForce="0,0,0") {
    angularDampingOverride=1
    angularSpringConstant=20000
    bAngularSpring=true
    bOverrideAngularDamping=true
    bUseGravityOverride=true
    gravityScale=0.35
    simulationSpace=BoneRelative
    alphaBoolBlend {blendOption=Linear}
    boundBone {boneName=dyn_shoulderArmor_base_l_02}
    componentAppliedLinearAccClamp=<1000 1000 1000>
    componentLinearAccScale=<0.3 0.2 0.2>
    componentLinearVelScale=<0.3 0.2 0.2>
    gravityOverride {z=-980}
    physicsBodyDefinitions [{
      boundBone {boneName=dyn_shoulderArmor_base_l_02}
      boxExtents=<20 30 30>
      constraintSetup {angularLimitsMax=<0 7 2>, angularLimitsMin=<0 -7 -2>, angularTarget=<1 0 0>}
      localJointOffset=<4 0 0>
     }]
    relativeSpaceBone {boneName=spine_05}
   }
   n25 = Anim Dynamics - Bone: dyn_shoulderArmor_base_l_03(ComponentPose=@n24, Alpha=1, ExternalForce="0,0,0") {
    angularDampingOverride=1
    angularSpringConstant=20000
    bAngularSpring=true
    bOverrideAngularDamping=true
    bUseGravityOverride=true
    gravityScale=0.35
    simulationSpace=BoneRelative
    alphaBoolBlend {blendOption=Linear}
    boundBone {boneName=dyn_shoulderArmor_base_l_03}
    componentAppliedLinearAccClamp=<1000 1000 1000>
    componentLinearAccScale=<0.3 0.2 0.2>
    componentLinearVelScale=<0.3 0.2 0.2>
    gravityOverride {z=-980}
    physicsBodyDefinitions [{
      boundBone {boneName=dyn_shoulderArmor_base_l_03}
      boxExtents=<20 30 30>
      constraintSetup {angularLimitsMax=<0 5 2>, angularLimitsMin=<0 -5 -2>, angularTarget=<1 0 0>}
      localJointOffset=<4 0 0>
     }]
    relativeSpaceBone {boneName=spine_05}
   }
   n26 = Anim Dynamics - Bone: dyn_shoulderArmor_base_r_01(ComponentPose=@n25, Alpha=1, ExternalForce="0,0,0") {
    angularDampingOverride=1
    angularSpringConstant=20000
    bAngularSpring=true
    bOverrideAngularDamping=true
    bUseGravityOverride=true
    gravityScale=0.35
    simulationSpace=BoneRelative
    alphaBoolBlend {blendOption=Linear}
    boundBone {boneName=dyn_shoulderArmor_base_r_01}
    componentAppliedLinearAccClamp=<1000 1000 1000>
    componentLinearAccScale=<0.3 0.2 0.2>
    componentLinearVelScale=<0.3 0.2 0.2>
    gravityOverride {z=-980}
    physicsBodyDefinitions [{
      boundBone {boneName=dyn_shoulderArmor_base_r_01}
      boxExtents=<20 30 30>
      constraintSetup {angularLimitsMax=<2 0 10>, angularLimitsMin=<-2 0 -10>, angularTarget=<1 0 0>}
      localJointOffset=<4 0 0>
     }]
    relativeSpaceBone {boneName=spine_05}
   }
   n27 = Anim Dynamics - Bone: dyn_shoulderArmor_base_r_02(ComponentPose=@n26, Alpha=1, ExternalForce="0,0,0") {
    angularDampingOverride=1
    angularSpringConstant=20000
    bAngularSpring=true
    bOverrideAngularDamping=true
    bUseGravityOverride=true
    gravityScale=0.35
    simulationSpace=BoneRelative
    alphaBoolBlend {blendOption=Linear}
    boundBone {boneName=dyn_shoulderArmor_base_r_02}
    componentAppliedLinearAccClamp=<1000 1000 1000>
    componentLinearAccScale=<0.3 0.2 0.2>
    componentLinearVelScale=<0.3 0.2 0.2>
    gravityOverride {z=-980}
    physicsBodyDefinitions [{
      boundBone {boneName=dyn_shoulderArmor_base_r_02}
      boxExtents=<20 30 30>
      constraintSetup {angularLimitsMax=<2 0 7>, angularLimitsMin=<-2 0 -7>, angularTarget=<1 0 0>}
      localJointOffset=<4 0 0>
     }]
    relativeSpaceBone {boneName=spine_05}
   }
   n28 = Anim Dynamics - Bone: dyn_shoulderArmor_base_r_03(ComponentPose=@n27, Alpha=1, ExternalForce="0,0,0") {
    angularDampingOverride=1
    angularSpringConstant=20000
    bAngularSpring=true
    bOverrideAngularDamping=true
    bUseGravityOverride=true
    gravityScale=0.35
    simulationSpace=BoneRelative
    alphaBoolBlend {blendOption=Linear}
    boundBone {boneName=dyn_shoulderArmor_base_r_03}
    componentAppliedLinearAccClamp=<1000 1000 1000>
    componentLinearAccScale=<0.3 0.2 0.2>
    componentLinearVelScale=<0.3 0.2 0.2>
    gravityOverride {z=-980}
    physicsBodyDefinitions [{
      boundBone {boneName=dyn_shoulderArmor_base_r_03}
      boxExtents=<20 30 30>
      constraintSetup {angularLimitsMax=<2 0 5>, angularLimitsMin=<-2 0 -5>, angularTarget=<1 0 0>}
      localJointOffset=<4 0 0>
     }]
    relativeSpaceBone {boneName=spine_05}
   }
   n29 = "Transform (Modify) Bone - Bone: dyn_shoulderArmor_master_r_Offset"(ComponentPose=@n28, Alpha=Get ShoulderPadOffset_LookUp_R(), Translation="0,-3,-5", Rotation="10,40,50", Scale="0,0,0") {
    rotationMode=BMM_Additive
    rotationSpace=BCS_BoneSpace
    scaleSpace=BCS_BoneSpace
    translationMode=BMM_Additive
    translationSpace=BCS_BoneSpace
    alphaBoolBlend {blendOption=Linear}
    boneToModify {boneName=dyn_shoulderArmor_master_r_Offset}
    rotation=<rot 10 40 50>
    scale=<0 0 0>
    translation {y=-3, z=-5}
   }
   n30 = "Transform (Modify) Bone - Bone: dyn_shoulderArmor_master_l_Offset"(ComponentPose=@n29, Alpha=Get ShoulderPadOffset_LookUp_L(), Translation="0,0,0", Rotation="25,-20,-25", Scale="0,0,0") {
    rotationMode=BMM_Additive
    rotationSpace=BCS_BoneSpace
    scaleSpace=BCS_BoneSpace
    translationMode=BMM_Additive
    translationSpace=BCS_BoneSpace
    alphaBoolBlend {blendOption=Linear}
    boneToModify {boneName=dyn_shoulderArmor_master_l_Offset}
    rotation=<rot 25 -20 -25>
    scale=<0 0 0>
   }
   n31 = "Transform (Modify) Bone - Bone: dyn_shoulderArmor_master_r_Offset"(ComponentPose=@n30, Alpha=1, Translation="0,0.5,0", Rotation="0,3,0", Scale="0,0,0") {
    rotationMode=BMM_Additive
    rotationSpace=BCS_BoneSpace
    scaleSpace=BCS_BoneSpace
    translationMode=BMM_Additive
    translationSpace=BCS_BoneSpace
    alphaBoolBlend {blendOption=Linear}
    boneToModify {boneName=dyn_shoulderArmor_master_r_Offset}
    rotation {yaw=3}
    scale=<0 0 0>
    translation {y=0.5}
   }
   n32 = "Transform (Modify) Bone - Bone: dyn_shoulderArmor_master_l_Offset"(ComponentPose=@n31, Alpha=1, Translation="0,0,0.05", Rotation="7,0,0", Scale="0,0,0") {
    rotationMode=BMM_Additive
    rotationSpace=BCS_BoneSpace
    scaleSpace=BCS_BoneSpace
    translationMode=BMM_Additive
    translationSpace=BCS_BoneSpace
    alphaBoolBlend {blendOption=Linear}
    boneToModify {boneName=dyn_shoulderArmor_master_l_Offset}
    rotation {pitch=7}
    scale=<0 0 0>
    translation {z=0.05}
   }
   n33 = Get Not_Crouching()
   n34 = "Transform (Modify) Bone - Bone: dyn_back_loincloth_01"(ComponentPose=@n32, Alpha=@n33, Translation="0,0,-2.5", Rotation="0,0,0", Scale="1,1,1") {
    rotationMode=BMM_Additive
    translationMode=BMM_Additive
    alphaBoolBlend {blendOption=Linear}
    boneToModify {boneName=dyn_back_loincloth_01}
    translation {z=-2.5}
   }
   n35 = "Transform (Modify) Bone - Bone: dyn_hip_armor_mstr_r_offset"(ComponentPose=@n34, Alpha=@n33, Translation="0,0,2", Rotation="0,0,0", Scale="1,1,1") {
    rotationMode=BMM_Additive
    translationMode=BMM_Additive
    alphaBoolBlend {blendOption=Linear}
    boneToModify {boneName=dyn_hip_armor_mstr_r_offset}
    translation {z=2}
   }
   n36 = "Transform (Modify) Bone - Bone: dyn_hip_armor_mstr_l_offset"(ComponentPose=@n35, Alpha=@n33, Translation="0,0,2", Rotation="0,0,0", Scale="1,1,1") {
    rotationMode=BMM_Additive
    translationMode=BMM_Additive
    alphaBoolBlend {blendOption=Linear}
    boneToModify {boneName=dyn_hip_armor_mstr_l_offset}
    translation {z=2}
   }
   n37 = Get LOD_Switch()
   n38 = "Constraint - dyn_shoulderArmor_master_l_Offset (PP)"(
    ComponentPose=Local To Component(LocalPose=Use cached pose 'CachedLegs'())
    Alpha=@n37
    ConstraintWeights_0=0.7779
    ConstraintWeights_1=0.2221
   ) {
    alphaBoolBlend {blendOption=Linear}
    boneToModify {boneName=dyn_shoulderArmor_master_l_Offset}
    constraintSetup [{targetBone {boneName=upperarm_twist_01_l}}, {targetBone {boneName=clavicle_l}}]
    constraintWeights [1,1]
   }
   n39 = "Constraint - dyn_shoulderArmor_master_r_Offset (PP)"(ComponentPose=@n38, Alpha=@n37, ConstraintWeights_0=0.7779, ConstraintWeights_1=0.2221) {
    alphaBoolBlend {blendOption=Linear}
    boneToModify {boneName=dyn_shoulderArmor_master_r_Offset}
    constraintSetup [{targetBone {boneName=upperarm_twist_01_r}}, {targetBone {boneName=clavicle_r}}]
    constraintWeights [1,1]
   }
   n40 = Anim Dynamics - Bone: dyn_shoulderArmor_base_l_01(ComponentPose=@n39, Alpha=1, ExternalForce="0,0,0") {
    angularDampingOverride=1
    angularSpringConstant=20000
    bAngularSpring=true
    bOverrideAngularDamping=true
    bUseGravityOverride=true
    gravityScale=0.35
    simulationSpace=BoneRelative
    alphaBoolBlend {blendOption=Linear}
    boundBone {boneName=dyn_shoulderArmor_base_l_01}
    componentAppliedLinearAccClamp=<1000 1000 1000>
    componentLinearAccScale=<0.3 0.2 0.2>
    componentLinearVelScale=<0.3 0.2 0.2>
    gravityOverride {z=-980}
    physicsBodyDefinitions [{
      boundBone {boneName=dyn_shoulderArmor_base_l_01}
      boxExtents=<20 30 30>
      constraintSetup {angularLimitsMax=<0 10 2>, angularLimitsMin=<0 -10 -2>, angularTarget=<1 0 0>}
      localJointOffset=<4 0 0>
     }]
    relativeSpaceBone {boneName=spine_05}
   }
   n41 = Anim Dynamics - Bone: dyn_shoulderArmor_base_l_02(ComponentPose=@n40, Alpha=1, ExternalForce="0,0,0") {
    angularDampingOverride=1
    angularSpringConstant=20000
    bAngularSpring=true
    bOverrideAngularDamping=true
    bUseGravityOverride=true
    gravityScale=0.35
    simulationSpace=BoneRelative
    alphaBoolBlend {blendOption=Linear}
    boundBone {boneName=dyn_shoulderArmor_base_l_02}
    componentAppliedLinearAccClamp=<1000 1000 1000>
    componentLinearAccScale=<0.3 0.2 0.2>
    componentLinearVelScale=<0.3 0.2 0.2>
    gravityOverride {z=-980}
    physicsBodyDefinitions [{
      boundBone {boneName=dyn_shoulderArmor_base_l_02}
      boxExtents=<20 30 30>
      constraintSetup {angularLimitsMax=<0 7 2>, angularLimitsMin=<0 -7 -2>, angularTarget=<1 0 0>}
      localJointOffset=<4 0 0>
     }]
    relativeSpaceBone {boneName=spine_05}
   }
   n42 = Anim Dynamics - Bone: dyn_shoulderArmor_base_l_03(ComponentPose=@n41, Alpha=1, ExternalForce="0,0,0") {
    angularDampingOverride=1
    angularSpringConstant=20000
    bAngularSpring=true
    bOverrideAngularDamping=true
    bUseGravityOverride=true
    gravityScale=0.35
    simulationSpace=BoneRelative
    alphaBoolBlend {blendOption=Linear}
    boundBone {boneName=dyn_shoulderArmor_base_l_03}
    componentAppliedLinearAccClamp=<1000 1000 1000>
    componentLinearAccScale=<0.3 0.2 0.2>
    componentLinearVelScale=<0.3 0.2 0.2>
    gravityOverride {z=-980}
    physicsBodyDefinitions [{
      boundBone {boneName=dyn_shoulderArmor_base_l_03}
      boxExtents=<20 30 30>
      constraintSetup {angularLimitsMax=<0 5 2>, angularLimitsMin=<0 -5 -2>, angularTarget=<1 0 0>}
      localJointOffset=<4 0 0>
     }]
    relativeSpaceBone {boneName=spine_05}
   }
   n43 = Anim Dynamics - Bone: dyn_shoulderArmor_base_r_01(ComponentPose=@n42, Alpha=1, ExternalForce="0,0,0") {
    angularDampingOverride=1
    angularSpringConstant=20000
    bAngularSpring=true
    bOverrideAngularDamping=true
    bUseGravityOverride=true
    gravityScale=0.35
    simulationSpace=BoneRelative
    alphaBoolBlend {blendOption=Linear}
    boundBone {boneName=dyn_shoulderArmor_base_r_01}
    componentAppliedLinearAccClamp=<1000 1000 1000>
    componentLinearAccScale=<0.3 0.2 0.2>
    componentLinearVelScale=<0.3 0.2 0.2>
    gravityOverride {z=-980}
    physicsBodyDefinitions [{
      boundBone {boneName=dyn_shoulderArmor_base_r_01}
      boxExtents=<20 30 30>
      constraintSetup {angularLimitsMax=<2 0 10>, angularLimitsMin=<-2 0 -10>, angularTarget=<1 0 0>}
      localJointOffset=<4 0 0>
     }]
    relativeSpaceBone {boneName=spine_05}
   }
   n44 = Anim Dynamics - Bone: dyn_shoulderArmor_base_r_02(ComponentPose=@n43, Alpha=1, ExternalForce="0,0,0") {
    angularDampingOverride=1
    angularSpringConstant=20000
    bAngularSpring=true
    bOverrideAngularDamping=true
    bUseGravityOverride=true
    gravityScale=0.35
    simulationSpace=BoneRelative
    alphaBoolBlend {blendOption=Linear}
    boundBone {boneName=dyn_shoulderArmor_base_r_02}
    componentAppliedLinearAccClamp=<1000 1000 1000>
    componentLinearAccScale=<0.3 0.2 0.2>
    componentLinearVelScale=<0.3 0.2 0.2>
    gravityOverride {z=-980}
    physicsBodyDefinitions [{
      boundBone {boneName=dyn_shoulderArmor_base_r_02}
      boxExtents=<20 30 30>
      constraintSetup {angularLimitsMax=<2 0 7>, angularLimitsMin=<-2 0 -7>, angularTarget=<1 0 0>}
      localJointOffset=<4 0 0>
     }]
    relativeSpaceBone {boneName=spine_05}
   }
   n45 = Anim Dynamics - Bone: dyn_shoulderArmor_base_r_03(ComponentPose=@n44, Alpha=1, ExternalForce="0,0,0") {
    angularDampingOverride=1
    angularSpringConstant=20000
    bAngularSpring=true
    bOverrideAngularDamping=true
    bUseGravityOverride=true
    gravityScale=0.35
    simulationSpace=BoneRelative
    alphaBoolBlend {blendOption=Linear}
    boundBone {boneName=dyn_shoulderArmor_base_r_03}
    componentAppliedLinearAccClamp=<1000 1000 1000>
    componentLinearAccScale=<0.3 0.2 0.2>
    componentLinearVelScale=<0.3 0.2 0.2>
    gravityOverride {z=-980}
    physicsBodyDefinitions [{
      boundBone {boneName=dyn_shoulderArmor_base_r_03}
      boxExtents=<20 30 30>
      constraintSetup {angularLimitsMax=<2 0 5>, angularLimitsMin=<-2 0 -5>, angularTarget=<1 0 0>}
      localJointOffset=<4 0 0>
     }]
    relativeSpaceBone {boneName=spine_05}
   }
   n46 = "Transform (Modify) Bone - Bone: dyn_shoulderArmor_master_r_Offset"(ComponentPose=@n45, Alpha=1, Translation="0,0.5,0", Rotation="0,3,0", Scale="0,0,0") {
    rotationMode=BMM_Additive
    rotationSpace=BCS_BoneSpace
    scaleSpace=BCS_BoneSpace
    translationMode=BMM_Additive
    translationSpace=BCS_BoneSpace
    alphaBoolBlend {blendOption=Linear}
    boneToModify {boneName=dyn_shoulderArmor_master_r_Offset}
    rotation {yaw=3}
    scale=<0 0 0>
    translation {y=0.5}
   }
   n47 = "Transform (Modify) Bone - Bone: dyn_shoulderArmor_master_l_Offset"(ComponentPose=@n46, Alpha=1, Translation="0,0,0.05", Rotation="7,0,0", Scale="0,0,0") {
    rotationMode=BMM_Additive
    rotationSpace=BCS_BoneSpace
    scaleSpace=BCS_BoneSpace
    translationMode=BMM_Additive
    translationSpace=BCS_BoneSpace
    alphaBoolBlend {blendOption=Linear}
    boneToModify {boneName=dyn_shoulderArmor_master_l_Offset}
    rotation {pitch=7}
    scale=<0 0 0>
    translation {z=0.05}
   }
   n48 = Blend Poses by bool(
    BlendPose_0=Component To Local(ComponentPose=@n47)
    BlendPose_1=Blend Poses by bool(BlendPose_0=Component To Local(ComponentPose=@n36), BlendPose_1=Component To Local(ComponentPose=@n19), bActiveValue=Get bIsSprinting()) {blendType=Linear}
    bActiveValue=Get bIsParachuteOpened()
   ) {blendType=Linear}
   n49 = Get LOD_Switch()
   n50 = "Constraint - dyn_shoulderArmor_master_l_Offset (PP)"(
    ComponentPose=Local To Component(LocalPose=Use cached pose 'CachedLegs'())
    Alpha=@n49
    ConstraintWeights_0=0.7779
    ConstraintWeights_1=0.2221
   ) {
    alphaBoolBlend {blendOption=Linear}
    boneToModify {boneName=dyn_shoulderArmor_master_l_Offset}
    constraintSetup [{targetBone {boneName=upperarm_twist_01_l}}, {targetBone {boneName=clavicle_l}}]
    constraintWeights [1,1]
   }
   n51 = "Constraint - dyn_shoulderArmor_master_r_Offset (PP)"(ComponentPose=@n50, Alpha=@n49, ConstraintWeights_0=0.7779, ConstraintWeights_1=0.2221) {
    alphaBoolBlend {blendOption=Linear}
    boneToModify {boneName=dyn_shoulderArmor_master_r_Offset}
    constraintSetup [{targetBone {boneName=upperarm_twist_01_r}}, {targetBone {boneName=clavicle_r}}]
    constraintWeights [1,1]
   }
   n52 = Anim Dynamics - Bone: dyn_shoulderArmor_base_l_01(ComponentPose=@n51, Alpha=1, ExternalForce="0,0,0") {
    angularDampingOverride=1
    angularSpringConstant=20000
    bAngularSpring=true
    bOverrideAngularDamping=true
    bUseGravityOverride=true
    gravityScale=0.35
    simulationSpace=BoneRelative
    alphaBoolBlend {blendOption=Linear}
    boundBone {boneName=dyn_shoulderArmor_base_l_01}
    componentAppliedLinearAccClamp=<1000 1000 1000>
    componentLinearAccScale=<0.3 0.2 0.2>
    componentLinearVelScale=<0.3 0.2 0.2>
    gravityOverride {z=-980}
    physicsBodyDefinitions [{
      boundBone {boneName=dyn_shoulderArmor_base_l_01}
      boxExtents=<20 30 30>
      constraintSetup {angularLimitsMax=<0 10 2>, angularLimitsMin=<0 -10 -2>, angularTarget=<1 0 0>}
      localJointOffset=<4 0 0>
     }]
    relativeSpaceBone {boneName=spine_05}
   }
   n53 = Anim Dynamics - Bone: dyn_shoulderArmor_base_l_02(ComponentPose=@n52, Alpha=1, ExternalForce="0,0,0") {
    angularDampingOverride=1
    angularSpringConstant=20000
    bAngularSpring=true
    bOverrideAngularDamping=true
    bUseGravityOverride=true
    gravityScale=0.35
    simulationSpace=BoneRelative
    alphaBoolBlend {blendOption=Linear}
    boundBone {boneName=dyn_shoulderArmor_base_l_02}
    componentAppliedLinearAccClamp=<1000 1000 1000>
    componentLinearAccScale=<0.3 0.2 0.2>
    componentLinearVelScale=<0.3 0.2 0.2>
    gravityOverride {z=-980}
    physicsBodyDefinitions [{
      boundBone {boneName=dyn_shoulderArmor_base_l_02}
      boxExtents=<20 30 30>
      constraintSetup {angularLimitsMax=<0 7 2>, angularLimitsMin=<0 -7 -2>, angularTarget=<1 0 0>}
      localJointOffset=<4 0 0>
     }]
    relativeSpaceBone {boneName=spine_05}
   }
   n54 = Anim Dynamics - Bone: dyn_shoulderArmor_base_l_03(ComponentPose=@n53, Alpha=1, ExternalForce="0,0,0") {
    angularDampingOverride=1
    angularSpringConstant=20000
    bAngularSpring=true
    bOverrideAngularDamping=true
    bUseGravityOverride=true
    gravityScale=0.35
    simulationSpace=BoneRelative
    alphaBoolBlend {blendOption=Linear}
    boundBone {boneName=dyn_shoulderArmor_base_l_03}
    componentAppliedLinearAccClamp=<1000 1000 1000>
    componentLinearAccScale=<0.3 0.2 0.2>
    componentLinearVelScale=<0.3 0.2 0.2>
    gravityOverride {z=-980}
    physicsBodyDefinitions [{
      boundBone {boneName=dyn_shoulderArmor_base_l_03}
      boxExtents=<20 30 30>
      constraintSetup {angularLimitsMax=<0 5 2>, angularLimitsMin=<0 -5 -2>, angularTarget=<1 0 0>}
      localJointOffset=<4 0 0>
     }]
    relativeSpaceBone {boneName=spine_05}
   }
   n55 = "Transform (Modify) Bone - Bone: dyn_shoulderArmor_master_r_Offset"(ComponentPose=@n54, Alpha=Get ShoulderPadOffset_LookUp_R(), Translation="0,-3,-5", Rotation="10,40,50", Scale="0,0,0") {
    rotationMode=BMM_Additive
    rotationSpace=BCS_BoneSpace
    scaleSpace=BCS_BoneSpace
    translationMode=BMM_Additive
    translationSpace=BCS_BoneSpace
    alphaBoolBlend {blendOption=Linear}
    boneToModify {boneName=dyn_shoulderArmor_master_r_Offset}
    rotation=<rot 10 40 50>
    scale=<0 0 0>
    translation {y=-3, z=-5}
   }
   n56 = "Transform (Modify) Bone - Bone: dyn_shoulderArmor_master_l_Offset"(ComponentPose=@n55, Alpha=Get ShoulderPadOffset_LookUp_L(), Translation="0,0,0", Rotation="25,-20,-25", Scale="0,0,0") {
    rotationMode=BMM_Additive
    rotationSpace=BCS_BoneSpace
    scaleSpace=BCS_BoneSpace
    translationMode=BMM_Additive
    translationSpace=BCS_BoneSpace
    alphaBoolBlend {blendOption=Linear}
    boneToModify {boneName=dyn_shoulderArmor_master_l_Offset}
    rotation=<rot 25 -20 -25>
    scale=<0 0 0>
   }
   n57 = "Transform (Modify) Bone - Bone: dyn_shoulderArmor_master_r_Offset"(ComponentPose=@n56, Alpha=1, Translation="0,0.5,0", Rotation="0,3,0", Scale="0,0,0") {
    rotationMode=BMM_Additive
    rotationSpace=BCS_BoneSpace
    scaleSpace=BCS_BoneSpace
    translationMode=BMM_Additive
    translationSpace=BCS_BoneSpace
    alphaBoolBlend {blendOption=Linear}
    boneToModify {boneName=dyn_shoulderArmor_master_r_Offset}
    rotation {yaw=3}
    scale=<0 0 0>
    translation {y=0.5}
   }
   n58 = "Transform (Modify) Bone - Bone: dyn_shoulderArmor_master_l_Offset"(ComponentPose=@n57, Alpha=1, Translation="0,0,0.05", Rotation="7,0,0", Scale="0,0,0") {
    rotationMode=BMM_Additive
    rotationSpace=BCS_BoneSpace
    scaleSpace=BCS_BoneSpace
    translationMode=BMM_Additive
    translationSpace=BCS_BoneSpace
    alphaBoolBlend {blendOption=Linear}
    boneToModify {boneName=dyn_shoulderArmor_master_l_Offset}
    rotation {pitch=7}
    scale=<0 0 0>
    translation {z=0.05}
   }
   n59 = "Transform (Modify) Bone - Bone: dyn_shoulderArmor_master_r_Offset"(ComponentPose=@n58, Alpha=1, Translation="0,0,0", Rotation="-15,0,0", Scale="1,1,1") {
    rotationMode=BMM_Additive
    translationMode=BMM_Additive
    alphaBoolBlend {blendOption=Linear}
    boneToModify {boneName=dyn_shoulderArmor_master_r_Offset}
    rotation {pitch=-15}
   }
   n60 = "Transform (Modify) Bone - Bone: dyn_shoulderArmor_master_l_Offset"(ComponentPose=@n59, Alpha=1, Translation="0,0,0", Rotation="15,0,0", Scale="1,1,1") {
    rotationMode=BMM_Additive
    translationMode=BMM_Additive
    alphaBoolBlend {blendOption=Linear}
    boneToModify {boneName=dyn_shoulderArmor_master_l_Offset}
    rotation {pitch=15}
   }
   n61 = Component To Local(ComponentPose=@n60)
   n62 = AND Boolean(A=Get bIsSkyDiving(), B=NOT Boolean(A=Get bIsParachuteOpened()))
   n63 = "Transform (Modify) Bone - Bone: dyn_hip_armor_mstr_l_offset"(
    ComponentPose=Local To Component(LocalPose=Blend Poses by bool(BlendPose_0=@n61, BlendPose_1=@n48, bActiveValue=@n62) {blendType=Linear})
    Alpha=Get HipArmor_Raise_L() -> (HipArmor_Raise_L:double)
    Translation="0,5,1"
    Rotation="0,0,20"
    Scale="0,0,0"
   ) {
    rotationMode=BMM_Additive
    rotationSpace=BCS_BoneSpace
    scaleSpace=BCS_BoneSpace
    translationMode=BMM_Additive
    translationSpace=BCS_BoneSpace
    alphaBoolBlend {blendOption=Linear}
    boneToModify {boneName=dyn_hip_armor_mstr_l_offset}
    rotation {roll=20}
    scale=<0 0 0>
    translation {y=5, z=1}
   }
   n64 = "Transform (Modify) Bone - Bone: dyn_hip_armor_mstr_r_offset"(ComponentPose=@n63, Alpha=Get HipArmor_Raise_R() -> (HipArmor_Raise_R:double), Translation="0,0,2", Rotation="0,0,-5", Scale="0,0,0") {
    rotationMode=BMM_Additive
    rotationSpace=BCS_BoneSpace
    scaleSpace=BCS_BoneSpace
    translationMode=BMM_Additive
    translationSpace=BCS_BoneSpace
    alphaBoolBlend {blendOption=Linear}
    boneToModify {boneName=dyn_hip_armor_mstr_r_offset}
    rotation {roll=-5}
    scale=<0 0 0>
    translation {z=2}
   }
   n65 = "To Float (Boolean)"(InBool=Get bIsParachuteOpened())
   n66 = "Transform (Modify) Bone - Bone: dyn_shoulderArmor_master_r_Offset"(ComponentPose=@n64, Alpha=@n65, Translation="-3,3,1", Rotation="-2,0,0", Scale="1,1,1") {
    rotationMode=BMM_Additive
    translationMode=BMM_Additive
    alphaBoolBlend {blendOption=Linear}
    boneToModify {boneName=dyn_shoulderArmor_master_r_Offset}
    rotation {pitch=-2}
    translation=<-3 3 1>
   }
   n67 = "Transform (Modify) Bone - Bone: dyn_shoulderArmor_master_l_Offset"(ComponentPose=@n66, Alpha=@n65, Translation="3,3,1", Rotation="2,0,0", Scale="1,1,1") {
    rotationMode=BMM_Additive
    translationMode=BMM_Additive
    alphaBoolBlend {blendOption=Linear}
    boneToModify {boneName=dyn_shoulderArmor_master_l_Offset}
    rotation {pitch=2}
    translation=<3 3 1>
   }
   n68 = "To Float (Boolean)"(InBool=Get bIsSkyDiving())
   n69 = "Transform (Modify) Bone - Bone: dyn_shoulderArmor_master_r_Offset"(ComponentPose=@n67, Alpha=@n68, Translation="0,0,0", Rotation="-10,5,0", Scale="1,1,1") {
    rotationMode=BMM_Additive
    translationMode=BMM_Additive
    alphaBoolBlend {blendOption=Linear}
    boneToModify {boneName=dyn_shoulderArmor_master_r_Offset}
    rotation {pitch=-10, yaw=5}
   }
   n70 = "Transform (Modify) Bone - Bone: dyn_shoulderArmor_master_l_Offset"(ComponentPose=@n69, Alpha=@n68, Translation="0,0,0", Rotation="10,-5,0", Scale="1,1,1") {
    rotationMode=BMM_Additive
    translationMode=BMM_Additive
    alphaBoolBlend {blendOption=Linear}
    boneToModify {boneName=dyn_shoulderArmor_master_l_Offset}
    rotation {pitch=10, yaw=-5}
   }
   n71 = "To Float (Boolean)"(InBool=Get bIsSkydiveFloating())
   n72 = "Transform (Modify) Bone - Bone: dyn_shoulderArmor_master_r_Offset"(ComponentPose=@n70, Alpha=@n71, Translation="0,0,0", Rotation="20,-10,0", Scale="1,1,1") {
    rotationMode=BMM_Additive
    translationMode=BMM_Additive
    alphaBoolBlend {blendOption=Linear}
    boneToModify {boneName=dyn_shoulderArmor_master_r_Offset}
    rotation {pitch=20, yaw=-10}
   }
   n73 = "Transform (Modify) Bone - Bone: dyn_shoulderArmor_master_l_Offset"(ComponentPose=@n72, Alpha=@n71, Translation="0,0,0", Rotation="-20,10,0", Scale="1,1,1") {
    rotationMode=BMM_Additive
    translationMode=BMM_Additive
    alphaBoolBlend {blendOption=Linear}
    boneToModify {boneName=dyn_shoulderArmor_master_l_Offset}
    rotation {pitch=-20, yaw=10}
   }
   n74 = "To Float (Boolean)"(InBool=Get bIsSkydiveDiveMode())
   n75 = "Transform (Modify) Bone - Bone: dyn_shoulderArmor_master_r_Offset"(ComponentPose=@n73, Alpha=@n74, Translation="0,0,0", Rotation="20,0,0", Scale="1,1,1") {
    rotationMode=BMM_Additive
    translationMode=BMM_Additive
    alphaBoolBlend {blendOption=Linear}
    boneToModify {boneName=dyn_shoulderArmor_master_r_Offset}
    rotation {pitch=20}
   }
   n76 = "Transform (Modify) Bone - Bone: dyn_shoulderArmor_master_l_Offset"(ComponentPose=@n75, Alpha=@n74, Translation="0,0,0", Rotation="-20,0,0", Scale="1,1,1") {
    rotationMode=BMM_Additive
    translationMode=BMM_Additive
    alphaBoolBlend {blendOption=Linear}
    boneToModify {boneName=dyn_shoulderArmor_master_l_Offset}
    rotation {pitch=-20}
   }
   Output Pose(Result=Component To Local(ComponentPose=@n76)) {layerGroup=None}
   Save cached pose 'AnimCache'(Pose=Copy Pose From Mesh() {bCopyCurves=true, bUseAttachedParent=true}) {cacheName=AnimCache}
   n77 = "Transform (Modify) Bone - Bone: dyn_hip_armor_mstr_l_offset"(
    ComponentPose=Local To Component(LocalPose=Use cached pose 'Pre_Swimming_Legs'())
    Alpha=1
    Translation="0,0,0"
    Rotation="10,0,20"
    Scale="1,1,1"
   ) {
    rotationMode=BMM_Additive
    translationMode=BMM_Additive
    alphaBoolBlend {blendOption=Linear}
    boneToModify {boneName=dyn_hip_armor_mstr_l_offset}
    rotation {pitch=10, roll=20}
   }
   n78 = "Transform (Modify) Bone - Bone: dyn_hip_armor_mstr_r_offset"(ComponentPose=@n77, Alpha=1, Translation="0,0,0", Rotation="-10,0,20", Scale="1,1,1") {
    rotationMode=BMM_Additive
    translationMode=BMM_Additive
    alphaBoolBlend {blendOption=Linear}
    boneToModify {boneName=dyn_hip_armor_mstr_r_offset}
    rotation {pitch=-10, roll=20}
   }
   Save cached pose 'CachedLegs'(Pose=Blend Poses by bool(BlendPose_0=Component To Local(ComponentPose=@n78), BlendPose_1=Use cached pose 'Pre_Swimming_Legs'(), bActiveValue=Get bIsSurfaceSwimming()) {blendType=Linear}) {cacheName=CachedLegs}
   n79 = "Constraint - dyn_front_loincloth_01 (RR)"!(ComponentPose=Local To Component(LocalPose=Use cached pose 'AnimCache'()), Alpha=1, ConstraintWeights_0=0.5, ConstraintWeights_1:float=0.5) -> (Pose) {
    alphaBoolBlend {blendOption=Linear}
    boneToModify {boneName=dyn_front_loincloth_01}
    constraintSetup [{transformType=Rotation, targetBone {boneName=thigh_l}}, {transformType=Rotation, targetBone {boneName=thigh_r}}]
    constraintWeights [1,1]
   }
   n80 = "Constraint - dyn_hip_armor_mstr_r_offset (R)"(ComponentPose=@n79, Alpha=0.5, ConstraintWeights_0=1) {
    alphaBoolBlend {blendOption=Linear}
    boneToModify {boneName=dyn_hip_armor_mstr_r_offset}
    constraintSetup [{transformType=Rotation, targetBone {boneName=thigh_r}}]
    constraintWeights [1]
   }
   n81 = "Constraint - dyn_hip_armor_mstr_l_offset (R)"(ComponentPose=@n80, Alpha=0.5, ConstraintWeights_0=1) {
    alphaBoolBlend {blendOption=Linear}
    boneToModify {boneName=dyn_hip_armor_mstr_l_offset}
    constraintSetup [{transformType=Rotation, targetBone {boneName=thigh_l}}]
    constraintWeights [1]
   }
   n82 = "Constraint - dyn_hip_armor_mstr_l_offset (R)"(ComponentPose=@n81, Alpha=0.5, ConstraintWeights_0=1) {
    alphaBoolBlend {blendOption=Linear}
    boneToModify {boneName=dyn_hip_armor_mstr_l_offset}
    constraintSetup [{transformType=Rotation, targetBone {boneName=thigh_l}}]
    constraintWeights [1]
   }
   n83 = RigidBody(ComponentPose=@n82, Alpha=0.5) {
    bOverrideWorldGravity=true
    overridePhysicsAsset=&/Junk_Samurai_HipArmor_Physics
    simulationSpace=BaseBoneSpace
    alphaBoolBlend {blendOption=Linear}
    baseBoneRef {boneName=pelvis}
    componentAppliedLinearAccClamp=<1000 1000 1000>
    componentLinearAccScale=<2 1 1>
    componentLinearVelScale=<2 1 1>
    overrideWorldGravity {z=-980}
   }
   n84 = RigidBody(ComponentPose=@n83) {
    bOverrideWorldGravity=true
    overridePhysicsAsset=&/Junk_Samurai_LoinCloth_Physics
    simulationSpace=BaseBoneSpace
    alphaBoolBlend {blendOption=Linear}
    baseBoneRef {boneName=pelvis}
    componentAppliedLinearAccClamp=<1000 1000 1000>
    componentLinearAccScale=<2 1 1>
    componentLinearVelScale=<2 1 1>
    overrideWorldGravity {z=-1200}
   }
   n85 = "Transform (Modify) Bone - Bone: dyn_hip_armor_mstr_l_offset"(ComponentPose=@n84, Alpha=1, Translation="0,-1,0", Rotation="0,0,-8", Scale="1,1,1") {
    rotationMode=BMM_Additive
    translationMode=BMM_Additive
    alphaBoolBlend {blendOption=Linear}
    boneToModify {boneName=dyn_hip_armor_mstr_l_offset}
    rotation {roll=-8}
    translation {y=-1}
   }
   n86 = "Transform (Modify) Bone - Bone: dyn_hip_armor_mstr_r_offset"(ComponentPose=@n85, Alpha=1, Translation="0,0,0", Rotation="0,0,-5", Scale="1,1,1") {
    rotationMode=BMM_Additive
    translationMode=BMM_Additive
    alphaBoolBlend {blendOption=Linear}
    boneToModify {boneName=dyn_hip_armor_mstr_r_offset}
    rotation {roll=-5}
   }
   n87 = "Constraint - dyn_front_loincloth_01 (RR)"(ComponentPose=Local To Component(LocalPose=Use cached pose 'AnimCache'()), Alpha=1, ConstraintWeights_0=0.5, ConstraintWeights_1=0.5) {
    alphaBoolBlend {blendOption=Linear}
    boneToModify {boneName=dyn_front_loincloth_01}
    constraintSetup [{transformType=Rotation, targetBone {boneName=thigh_l}}, {transformType=Rotation, targetBone {boneName=thigh_r}}]
    constraintWeights [1,1]
   }
   n88 = "Constraint - dyn_hip_armor_mstr_r_offset (R)"(ComponentPose=@n87, Alpha=0.7, ConstraintWeights_0=1) {
    alphaBoolBlend {blendOption=Linear}
    boneToModify {boneName=dyn_hip_armor_mstr_r_offset}
    constraintSetup [{transformType=Rotation, targetBone {boneName=thigh_r}}]
    constraintWeights [1]
   }
   n89 = "Constraint - dyn_hip_armor_mstr_l_offset (R)"(ComponentPose=@n88, Alpha=0.7, ConstraintWeights_0=1) {
    alphaBoolBlend {blendOption=Linear}
    boneToModify {boneName=dyn_hip_armor_mstr_l_offset}
    constraintSetup [{transformType=Rotation, targetBone {boneName=thigh_l}}]
    constraintWeights [1]
   }
   n90 = "Constraint - dyn_hip_armor_mstr_l_offset (R)"(ComponentPose=@n89, Alpha=0.5, ConstraintWeights_0=1) {
    alphaBoolBlend {blendOption=Linear}
    boneToModify {boneName=dyn_hip_armor_mstr_l_offset}
    constraintSetup [{transformType=Rotation, targetBone {boneName=thigh_l}}]
    constraintWeights [1]
   }
   n91 = RigidBody(ComponentPose=@n90, Alpha=0.5) {
    bOverrideWorldGravity=true
    overridePhysicsAsset=&Junk_Samurai_HipArmor_Physics
    simulationSpace=BaseBoneSpace
    alphaBoolBlend {blendOption=Linear}
    baseBoneRef {boneName=pelvis}
    componentAppliedLinearAccClamp=<1000 1000 1000>
    componentLinearAccScale=<2 1 1>
    componentLinearVelScale=<2 1 1>
    overrideWorldGravity {z=-980}
   }
   n92 = RigidBody(ComponentPose=@n91) {
    bClampLinearTranslationLimitToRefPose=true
    bOverrideWorldGravity=true
    overridePhysicsAsset=&/Junk_Samurai_LoinCloth_Physics_Sprint
    simulationSpace=BaseBoneSpace
    alphaBoolBlend {blendOption=Linear}
    baseBoneRef {boneName=root}
    componentAppliedLinearAccClamp=<5000 5000 5000>
    componentLinearAccScale=<0.25 0.25 0.25>
    overrideWorldGravity {z=-980}
   }
   n93 = "Transform (Modify) Bone - Bone: dyn_hip_armor_mstr_l_offset"(ComponentPose=@n92, Alpha=1, Translation="0,0,0", Rotation="5,0,-15", Scale="1,1,1") {
    rotationMode=BMM_Additive
    translationMode=BMM_Additive
    alphaBoolBlend {blendOption=Linear}
    boneToModify {boneName=dyn_hip_armor_mstr_l_offset}
    rotation {pitch=5, roll=-15}
   }
   n94 = "Transform (Modify) Bone - Bone: dyn_hip_armor_mstr_r_offset"(ComponentPose=@n93, Alpha=1, Translation="0,0,0", Rotation="-5,0,-15", Scale="1,1,1") {
    rotationMode=BMM_Additive
    translationMode=BMM_Additive
    alphaBoolBlend {blendOption=Linear}
    boneToModify {boneName=dyn_hip_armor_mstr_r_offset}
    rotation {pitch=-5, roll=-15}
   }
   n95 = "Transform (Modify) Bone - Bone: dyn_front_loincloth_01"(ComponentPose=@n94, Alpha=1, Translation="0,10,0", Rotation="0,0,0", Scale="1,1,1") {
    rotationMode=BMM_Additive
    translationMode=BMM_Additive
    alphaBoolBlend {blendOption=Linear}
    boneToModify {boneName=dyn_front_loincloth_01}
    translation {y=10}
   }
   n96 = Blend Poses by bool(BlendPose_0=Component To Local(ComponentPose=@n95), BlendPose_1=Component To Local(ComponentPose=@n86), bActiveValue=Get bIsSprinting()) {blendType=Linear}
   n97 = "Constraint - dyn_front_loincloth_01 (RR)"(ComponentPose=Local To Component(LocalPose=Use cached pose 'AnimCache'()), Alpha=1, ConstraintWeights_0=0.5, ConstraintWeights_1=0.5) {
    alphaBoolBlend {blendOption=Linear}
    boneToModify {boneName=dyn_front_loincloth_01}
    constraintSetup [{transformType=Rotation, targetBone {boneName=thigh_l}}, {transformType=Rotation, targetBone {boneName=thigh_r}}]
    constraintWeights [1,1]
   }
   n98 = "Constraint - dyn_hip_armor_mstr_r_offset (R)"(ComponentPose=@n97, Alpha=1, ConstraintWeights_0=1) {
    alphaBoolBlend {blendOption=Linear}
    boneToModify {boneName=dyn_hip_armor_mstr_r_offset}
    constraintSetup [{transformType=Rotation, targetBone {boneName=thigh_r}}]
    constraintWeights [1]
   }
   n99 = "Constraint - dyn_hip_armor_mstr_l_offset (R)"(ComponentPose=@n98, Alpha=0.7, ConstraintWeights_0=1) {
    alphaBoolBlend {blendOption=Linear}
    boneToModify {boneName=dyn_hip_armor_mstr_l_offset}
    constraintSetup [{transformType=Rotation, targetBone {boneName=thigh_l}}]
    constraintWeights [1]
   }
   n100 = "Constraint - dyn_hip_armor_mstr_l_offset (R)"(ComponentPose=@n99, Alpha=0.5, ConstraintWeights_0=1) {
    alphaBoolBlend {blendOption=Linear}
    boneToModify {boneName=dyn_hip_armor_mstr_l_offset}
    constraintSetup [{transformType=Rotation, targetBone {boneName=thigh_l}}]
    constraintWeights [1]
   }
   n101 = RigidBody(ComponentPose=@n100, Alpha=0.3) {
    bOverrideWorldGravity=true
    overridePhysicsAsset=&/Junk_Samurai_HipArmor_Physics_Jump
    simulationSpace=BaseBoneSpace
    alphaBoolBlend {blendOption=Linear}
    baseBoneRef {boneName=pelvis}
    componentAppliedLinearAccClamp=<1000 1000 1000>
    componentLinearAccScale=<2 1 1>
    componentLinearVelScale=<2 1 1>
    overrideWorldGravity {z=-980}
   }
   n102 = RigidBody(ComponentPose=@n101) {
    bOverrideWorldGravity=true
    overridePhysicsAsset=&Junk_Samurai_LoinCloth_Physics
    simulationSpace=BaseBoneSpace
    alphaBoolBlend {blendOption=Linear}
    baseBoneRef {boneName=pelvis}
    componentAppliedLinearAccClamp=<1000 1000 1000>
    componentLinearAccScale=<2 1 1>
    componentLinearVelScale=<2 1 1>
    overrideWorldGravity {z=-1200}
   }
   n103 = "Transform (Modify) Bone - Bone: dyn_hip_armor_mstr_l_offset"(ComponentPose=@n102, Alpha=1, Translation="0,0,0", Rotation="5,0,-15", Scale="1,1,1") {
    rotationMode=BMM_Additive
    translationMode=BMM_Additive
    alphaBoolBlend {blendOption=Linear}
    boneToModify {boneName=dyn_hip_armor_mstr_l_offset}
    rotation {pitch=5, roll=-15}
   }
   n104 = "Transform (Modify) Bone - Bone: dyn_hip_armor_mstr_r_offset"(ComponentPose=@n103, Alpha=1, Translation="0,0,0", Rotation="-5,0,-15", Scale="1,1,1") {
    rotationMode=BMM_Additive
    translationMode=BMM_Additive
    alphaBoolBlend {blendOption=Linear}
    boneToModify {boneName=dyn_hip_armor_mstr_r_offset}
    rotation {pitch=-5, roll=-15}
   }
   n105 = "Transform (Modify) Bone - Bone: dyn_front_loincloth_01"(ComponentPose=@n104, Alpha=1, Translation="0,10,0", Rotation="0,0,-45", Scale="1,1,1") {
    rotationMode=BMM_Additive
    translationMode=BMM_Additive
    alphaBoolBlend {blendOption=Linear}
    boneToModify {boneName=dyn_front_loincloth_01}
    rotation {roll=-45}
    translation {y=10}
   }
   n106 = Component To Local(ComponentPose=@n105)
   n107 = "Constraint - dyn_front_loincloth_01 (RR)"(ComponentPose=Local To Component(LocalPose=Use cached pose 'AnimCache'()), Alpha=1, ConstraintWeights_0=0.5, ConstraintWeights_1=0.5) {
    alphaBoolBlend {blendOption=Linear}
    boneToModify {boneName=dyn_front_loincloth_01}
    constraintSetup [{transformType=Rotation, targetBone {boneName=thigh_l}}, {transformType=Rotation, targetBone {boneName=thigh_r}}]
    constraintWeights [1,1]
   }
   n108 = "Constraint - dyn_hip_armor_mstr_r_offset (R)"(ComponentPose=@n107, Alpha=1, ConstraintWeights_0=1) {
    alphaBoolBlend {blendOption=Linear}
    boneToModify {boneName=dyn_hip_armor_mstr_r_offset}
    constraintSetup [{transformType=Rotation, targetBone {boneName=thigh_r}}]
    constraintWeights [1]
   }
   n109 = "Constraint - dyn_hip_armor_mstr_l_offset (R)"(ComponentPose=@n108, Alpha=0.5, ConstraintWeights_0=1) {
    alphaBoolBlend {blendOption=Linear}
    boneToModify {boneName=dyn_hip_armor_mstr_l_offset}
    constraintSetup [{transformType=Rotation, targetBone {boneName=thigh_l}}]
    constraintWeights [1]
   }
   n110 = "Constraint - dyn_hip_armor_mstr_l_offset (R)"(ComponentPose=@n109, Alpha=0.5, ConstraintWeights_0=1) {
    alphaBoolBlend {blendOption=Linear}
    boneToModify {boneName=dyn_hip_armor_mstr_l_offset}
    constraintSetup [{transformType=Rotation, targetBone {boneName=thigh_l}}]
    constraintWeights [1]
   }
   n111 = RigidBody(ComponentPose=@n110, Alpha=0.3) {
    bOverrideWorldGravity=true
    overridePhysicsAsset=&Junk_Samurai_HipArmor_Physics_Jump
    simulationSpace=BaseBoneSpace
    alphaBoolBlend {blendOption=Linear}
    baseBoneRef {boneName=pelvis}
    componentAppliedLinearAccClamp=<1000 1000 1000>
    componentLinearAccScale=<2 1 1>
    componentLinearVelScale=<2 1 1>
    overrideWorldGravity {z=-980}
   }
   n112 = RigidBody(ComponentPose=@n111) {
    bOverrideWorldGravity=true
    overridePhysicsAsset=&Junk_Samurai_LoinCloth_Physics
    simulationSpace=BaseBoneSpace
    alphaBoolBlend {blendOption=Linear}
    baseBoneRef {boneName=pelvis}
    componentAppliedLinearAccClamp=<1000 1000 1000>
    componentLinearAccScale=<2 1 1>
    componentLinearVelScale=<2 1 1>
    overrideWorldGravity {z=-1200}
   }
   n113 = "Transform (Modify) Bone - Bone: dyn_hip_armor_mstr_l_offset"(ComponentPose=@n112, Alpha=1, Translation="0,0,0", Rotation="0,0,-8", Scale="1,1,1") {
    rotationMode=BMM_Additive
    translationMode=BMM_Additive
    alphaBoolBlend {blendOption=Linear}
    boneToModify {boneName=dyn_hip_armor_mstr_l_offset}
    rotation {roll=-8}
   }
   n114 = "Transform (Modify) Bone - Bone: dyn_hip_armor_mstr_r_offset"(ComponentPose=@n113, Alpha=1, Translation="0,0,0", Rotation="0,0,-5", Scale="1,1,1") {
    rotationMode=BMM_Additive
    translationMode=BMM_Additive
    alphaBoolBlend {blendOption=Linear}
    boneToModify {boneName=dyn_hip_armor_mstr_r_offset}
    rotation {roll=-5}
   }
   n115 = Component To Local(ComponentPose=@n114)
   n116 = Get bIsSprinting()
   n117 = "Transform (Modify) Bone - Bone: dyn_hip_armor_mstr_r_offset"(
    ComponentPose=Local To Component(LocalPose=Blend Poses by bool(BlendPose_0=@n106, BlendPose_1=@n115, bActiveValue=@n116) {blendType=Linear})
    Alpha=1
    Translation="-2,-5,2"
    Rotation="0,-20,-10"
    Scale="1,1,1"
   ) {
    rotationMode=BMM_Additive
    translationMode=BMM_Additive
    alphaBoolBlend {blendOption=Linear}
    boneToModify {boneName=dyn_hip_armor_mstr_r_offset}
    rotation {roll=-10, yaw=-20}
    translation=<-2 -5 2>
   }
   n118 = Component To Local(ComponentPose=@n117)
   n119 = OR Boolean(A=Get bIsJumping(), B=Get bIsFalling())
   n120 = "Transform (Modify) Bone - Bone: dyn_shoulderArmor_master_l_Offset"(ComponentPose=Local To Component(LocalPose=Use cached pose 'AnimCache'()), Alpha=1, Translation="0,0,0", Rotation="90,0,0", Scale="1,1,1") {
    rotationMode=BMM_Additive
    translationMode=BMM_Additive
    alphaBoolBlend {blendOption=Linear}
    boneToModify {boneName=dyn_shoulderArmor_master_l_Offset}
    rotation {pitch=90}
   }
   n121 = "Constraint - dyn_hip_armor_mstr_r_offset (P)"(ComponentPose=@n120, Alpha=1, ConstraintWeights_0=1) {
    alphaBoolBlend {blendOption=Linear}
    boneToModify {boneName=dyn_hip_armor_mstr_r_offset}
    constraintSetup [{targetBone {boneName=thigh_r}}]
    constraintWeights [1]
   }
   n122 = "Constraint - dyn_hip_armor_mstr_l_offset (R)"(ComponentPose=@n121, Alpha=1, ConstraintWeights_0=1) {
    alphaBoolBlend {blendOption=Linear}
    boneToModify {boneName=dyn_hip_armor_mstr_l_offset}
    constraintSetup [{transformType=Rotation, targetBone {boneName=thigh_l}}]
    constraintWeights [1]
   }
   n123 = "Constraint - dyn_hip_armor_mstr_l_offset (R)"(ComponentPose=@n122, Alpha=0.5, ConstraintWeights_0=1) {
    alphaBoolBlend {blendOption=Linear}
    boneToModify {boneName=dyn_hip_armor_mstr_l_offset}
    constraintSetup [{transformType=Rotation, targetBone {boneName=thigh_l}}]
    constraintWeights [1]
   }
   n124 = "Transform (Modify) Bone - Bone: dyn_hip_armor_mstr_l_offset"(ComponentPose=@n123, Alpha=1, Translation="1.5,-5,0", Rotation="10,20,-10", Scale="1,1,1") {
    rotationMode=BMM_Additive
    translationMode=BMM_Additive
    alphaBoolBlend {blendOption=Linear}
    boneToModify {boneName=dyn_hip_armor_mstr_l_offset}
    rotation=<rot 10 20 -10>
    translation=<1.5 -5>
   }
   n125 = "Transform (Modify) Bone - Bone: dyn_hip_armor_mstr_r_offset"(ComponentPose=@n124, Alpha=1, Translation="0,0,3.5", Rotation="0,0,-15", Scale="1,1,1") {
    rotationMode=BMM_Additive
    translationMode=BMM_Additive
    alphaBoolBlend {blendOption=Linear}
    boneToModify {boneName=dyn_hip_armor_mstr_r_offset}
    rotation {roll=-15}
    translation {z=3.5}
   }
   n126 = RigidBody(ComponentPose=@n125, Alpha=0.3) {
    bOverrideWorldGravity=true
    overridePhysicsAsset=&/Junk_Samurai_HipArmor_Physics_Crouch
    simulationSpace=BaseBoneSpace
    alphaBoolBlend {blendOption=Linear}
    baseBoneRef {boneName=pelvis}
    componentAppliedLinearAccClamp=<1000 1000 1000>
    componentLinearAccScale=<2 1 1>
    componentLinearVelScale=<2 1 1>
    overrideWorldGravity {z=-980}
   }
   n127 = RigidBody(ComponentPose=@n126) {
    bClampLinearTranslationLimitToRefPose=true
    bOverrideWorldGravity=true
    overridePhysicsAsset=&/Junk_Samurai_LoinCloth_Physics_Crouch
    simulationSpace=BaseBoneSpace
    alphaBoolBlend {blendOption=Linear}
    baseBoneRef {boneName=root}
    componentAppliedLinearAccClamp=<5000 5000 5000>
    componentLinearAccScale=<0.25 0.25 0.25>
    overrideWorldGravity {z=-980}
   }
   n128 = "Constraint - dyn_front_loincloth_01 (RR)"(ComponentPose=@n127, Alpha=0.25, ConstraintWeights_0=1, ConstraintWeights_1=1) {
    alphaBoolBlend {blendOption=Linear}
    boneToModify {boneName=dyn_front_loincloth_01}
    constraintSetup [{transformType=Rotation, targetBone {boneName=thigh_l}}, {transformType=Rotation, targetBone {boneName=thigh_r}}]
    constraintWeights [1,1]
   }
   Save cached pose 'Pre_Swimming_Legs'(Pose=Blend Poses by bool(
     BlendPose_0=Component To Local(ComponentPose=@n128)
     BlendPose_1=Blend Poses by bool(BlendPose_0=@n118, BlendPose_1=@n96, bActiveValue=@n119) {blendType=Linear}
     bActiveValue=Get bIsCrouching()
    ) {blendType=Linear}) {cacheName=Pre_Swimming_Legs}
  ]}
 EventGraph {nodes [
   Event BlueprintUpdateAnimation()
   n1 = BreakTransform(InTransform=GetDeltaTransformFromRefPose(self=GetOwnerSkeletalMeshComponent(), BoneName=upperarm_l, BaseName=clavicle_l))
   > Set ShoulderPadOffset_LookUp_L(ShoulderPadOffset_LookUp_L:double=MapRangeClamped(Value=BreakRotator(InRot=@n1.Rotation).Yaw, InRangeA=15, InRangeB=90)) -> (Output_Get:double)
   n2 = BreakTransform(InTransform=GetDeltaTransformFromRefPose(self=GetOwnerSkeletalMeshComponent(), BoneName=upperarm_r, BaseName=clavicle_r))
   > Set ShoulderPadOffset_LookUp_R(ShoulderPadOffset_LookUp_R:double=MapRangeClamped(Value=BreakRotator(InRot=@n2.Rotation).Yaw, InRangeA=0, InRangeB=120)) -> (Output_Get:double)
   n3 = BreakTransform(InTransform=GetDeltaTransformFromRefPose(self=GetOwnerSkeletalMeshComponent(), BoneName=upperarm_r, BaseName=clavicle_r))
   > Set ShoulderPadOffset_Relax_R(ShoulderPadOffset_Relax_R:double=MapRangeClamped(Value=BreakRotator(InRot=@n3.Rotation).Yaw, InRangeA=2, InRangeB=-20)) -> (Output_Get:double)
   n4 = BreakTransform(InTransform=GetDeltaTransformFromRefPose(self=GetOwnerSkeletalMeshComponent(), BoneName=thigh_l, BaseName=calf_l))
   > Set HipArmor_Raise_L(HipArmor_Raise_L:double=MapRangeClamped(Value=BreakRotator(InRot=@n4.Rotation).Yaw, InRangeA=-20, InRangeB=-106)) -> (Output_Get:double)
   n5 = BreakTransform(InTransform=GetDeltaTransformFromRefPose(self=GetOwnerSkeletalMeshComponent(), BoneName=thigh_r, BaseName=calf_r))
   > Set HipArmor_Raise_R(HipArmor_Raise_R:double=MapRangeClamped(Value=BreakRotator(InRot=@n5.Rotation).Yaw, InRangeA=-17, InRangeB=-106)) -> (Output_Get:double)
   n6 = Get CurrentLOD()
   > Set LOD_Switch(LOD_Switch:double=MapRangeClamped(Value="To Float (Integer)"(InInt=@n6), InRangeA=1, InRangeB=2, OutRangeA=1, OutRangeB=0)) -> (Output_Get:double)
   n7 = Get bIsCrouching()
   > Set Not_Crouching(Not_Crouching:double=SelectFloat(A=0, B=1, bPickA=@n7)) -> (Output_Get:double)
   OR Boolean(A=Get bIsSprinting(), B=@n7)
  ]}
]
variables [
 HipArmor_Raise_L {type=number}
 HipArmor_Raise_R {type=number}
 LargeGunOut {type=number}
 LOD_Switch {type=number}
 Not_Crouching {type=number}
 rifle_out {type=boolean}
 ShoulderPadOffset_LookUp_L {type=number}
 ShoulderPadOffset_LookUp_R {type=number}
 ShoulderPadOffset_Relax_R {type=number}
 SkydiveFloating {type=number}
]
